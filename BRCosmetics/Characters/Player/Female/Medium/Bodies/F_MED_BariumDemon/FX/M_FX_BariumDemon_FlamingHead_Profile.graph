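# Unlit, translucent, two-sided flame material for Niagara mesh particles, used by the BariumDemon cosmetic's flaming head (Profile variant).
# Two panning, distorted noise layers form MainNoise, which drives eroded opacity with depth fade on higher quality levels.
# Emissive blends base and bright colors using a particle-centered sphere mask, plus hotspots.
# Vertex-color-masked world position offset combines scrolling noise, sine-driven rotation and a view-facing-dependent camera offset, with dynamic parameters feeding radius and rotation.
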Material /BRCosmetics/Characters/Player/Female/Medium/Bodies/F_MED_BariumDemon/FX/M_FX_BariumDemon_FlamingHead_Profile
graph {
 properties {
  bUsedWithNiagaraMeshParticles=true
  blendMode=BLEND_Translucent
  pixelDepthOffsetMode=PDOM_Legacy
  shadingModel=MSM_Unlit
  twoSided=true
  usedShadingModels=Unlit
 }
 nodes [
  Add(A=Multiply())
  n1 = VertexColor()
  n2 = Multiply(A=ScalarParameter() {defaultValue=1, group=WPO, parameterName=WPO_Tile}, B=NamedRerouteUsage())
  n3 = Multiply(A=Time(), B=ScalarParameter() {defaultValue=0.7, group=WPO, parameterName=WPO_Speed})
  n4 = TextureSampleParameter2D(UVs=Add(A=Panner(Coordinate=@n2, Time=@n3) {bFractionalPart=true, speedX=-1, speedY=0.2}, B=NamedRerouteUsage())) {parameterName=T_WPONoise, samplerType=SAMPLERTYPE_LinearColor, texture=/Game/Items/Traps/Textures/T_Plasma_Flames_02_M}
  n5 = Subtract(A=Multiply(A=@n4.G) {constB=2})
  n6 = Normalize(VectorInput=Transform(None=VectorParameter() {parameterName=WPO_Dir, defaultValue=<rgba 0 1 0 1>}.RGB) {transformSourceType=TRANSFORMSOURCE_Local})
  n7 = Multiply(
   A=ComponentMask(None=TextureCoordinate()) {r=true}
   B=Multiply(A=Multiply(A=@n5, B=@n6), B=ScalarParameter() {group=WPO, parameterName=WPO_Strength})
  )
  n8 = MaterialFunctionCall(Offset Amount=ScalarParameter() {group=WPO, parameterName=RearView_CameraOffset}) {materialFunction=/Engine/Functions/Engine_MaterialFunctions02/WorldPositionOffset/CameraOffset}
  n9 = Subtract(A=WorldPosition().XYZ, B=CameraPositionWS())
  n10 = Step(X=DotProduct(A=Normalize(VectorInput=@n9), B=VertexNormalWS())) {constX=0}
  Add(A=NamedRerouteUsage(), B=Add(A=Multiply(A=@n1.B, B=@n7), B=Multiply(A=@n8.Camera Offset, B=@n10)))
  Clamp(None=TwoSidedSign())
  n11 = TextureCoordinate()
  ComponentMask(None=@n11) {g=true}
  MaterialFunctionCall() {materialFunction=&CameraOffset}
  n12 = Multiply(
   A=SmoothStep(
    Min=ScalarParameter() {group=Hotspots, parameterName=Hotspots_Min}
    Max=ScalarParameter() {group=Hotspots, parameterName=Hotspots_Max}
    Value=NamedRerouteUsage()
   )
   B=ScalarParameter() {parameterName=Hotspots_Str}
  )
  n13 = VectorParameter() {parameterName=BrightColor, defaultValue=<rgba 0 0 0 1>}
  n14 = VectorParameter() {parameterName=BaseColor, defaultValue=<rgba 0 0 0 1>}
  n15 = ParticlePositionWS()
  n16 = WorldPosition()
  n17 = Multiply(
   A=DynamicParameter() {paramNames [SrcLightRadius, ParticleOffset, RotationMul, Param4]}.SrcLightRadius
   B=LinearInterpolate(Alpha=NamedRerouteUsage()) {constA=0.8, constB=1.1}
  )
  n18 = Add(A=Power(Base=SphereMask(A=@n15, B=@n16.XYZ, Radius=@n17) {hardnessPercent=0}), B=NamedRerouteUsage())
  MaterialFunctionCall(Emissive=Add(A=Multiply(A=@n12, B=@n13.RGB), B=LinearInterpolate(A=@n14.RGB, B=@n13.RGB, Alpha=@n18))) {materialFunction=&MF_FX_EmissiveAdjustments}
  n19 = ComponentMask(None=TextureCoordinate()) {r=true}
  n20 = ScalarParameter() {parameterName=gradient_shift}
  n21 = VertexColor()
  n22 = OneMinus(None=@n21.R)
  Multiply(A=Saturate(None=Add(A=@n19, B=@n20)), B=@n22)
  NamedRerouteDeclaration(None=Saturate(None=@n22)) {name=FlamesMask, nodeColor=<rgba 0.125 1 0 1>}
  n23 = Multiply(
   A=ComponentMask(None=VectorParameter() {parameterName=Overlay_Scale, defaultValue=<rgba 1 1 0 1>}.RGB) {g=true, r=true}
   B=NamedRerouteUsage()
  )
  n24 = Multiply(A=Time(), B=ScalarParameter() {defaultValue=0.7, group=Overlay, parameterName=OverlaySpeed})
  n25 = Subtract(A=Multiply(A=NamedRerouteUsage()) {constB=2})
  n26 = ScalarParameter() {parameterName=OverlayDistortion}
  n27 = TextureSampleParameter2D(UVs=Add(A=Panner(Coordinate=@n23, Time=@n24) {speedX=-1}, B=Multiply(A=@n25, B=@n26))) {parameterName=T_OverlayTex, samplerType=SAMPLERTYPE_LinearColor, texture=&T_Plasma_Flames_02_M}
  n28 = ScalarParameter() {defaultValue=1, parameterName=OverlayContrast}
  NamedRerouteDeclaration(None=Multiply(A=Power(Base=@n27.R, Exp=@n28), B=ScalarParameter() {defaultValue=1, parameterName=OverlayBrightness})) {name=Overlays, nodeColor=<rgba 0.1953 0 1 1>}
  NamedRerouteDeclaration(None=DynamicParameter() {paramNames [SrcLightRadius, ParticleOffset, RotationMul, Param4]}.ParticleOffset) {name=ParticleOffset, nodeColor=<rgba 0 1 0.5313 1>}
  NamedRerouteDeclaration(None=@n4.G) {name=WPOOffset, nodeColor=<rgba 0.2187 1 0 1>}
  n29 = Transform(None=Normalize(VectorInput=VectorParameter() {group=Rotating, parameterName=Rotation, defaultValue=<rgba 0 0 0 1>}.RGB)) {transformSourceType=TRANSFORMSOURCE_Local}
  n30 = Time()
  n31 = NamedRerouteUsage()
  n32 = Multiply(A=Sine(None=Add(A=@n30, B=@n31))) {constB=0.5}
  n33 = ScalarParameter() {group=Rotating, parameterName=Rotate}
  n34 = DynamicParameter() {paramNames [SrcLightRadius, ParticleOffset, RotationMul, Param4]}
  n35 = Multiply(A=Multiply(A=Add(A=@n32), B=Multiply(A=@n33, B=@n34.RotationMul)), B=ComponentMask(None=@n11) {r=true})
  n36 = ParticlePositionWS()
  n37 = WorldPosition()
  NamedRerouteDeclaration(None=Multiply(A=VertexColor().B, B=RotateAboutAxis(NormalizedRotationAxis=@n29, RotationAngle=@n35, PivotPoint=@n36, Position=@n37.XYZ))) {name=RotationalWPO, nodeColor=<rgba 0 0.7891 1 1>}
  OneMinus()
  OneMinus(None=ComponentMask(None=TextureCoordinate()) {r=true})
  Power(Base=@n21.R, Exp=ScalarParameter() {parameterName=rrr}) {constExponent=8}
  n38 = Add(A=NamedRerouteUsage(), B=ScalarParameter() {group=Erode, parameterName=Erode_Position})
  n39 = Add(A=@n38, B=ScalarParameter() {group=Erode, parameterName=Erode_W})
  n40 = TextureCoordinate()
  n41 = VectorParameter() {parameterName=UV_Scale, defaultValue=<rgba 1 1 0 1>}
  n42 = ComponentMask(None=@n41.RGB) {g=true, r=true}
  n43 = NamedRerouteDeclaration(None=Multiply(A=Multiply(A=@n40, B=@n42), B=@n41.B)) {name=ScaledUV, nodeColor=<rgba 0 0.3906 1 1>}
  n44 = NamedRerouteUsage()
  n45 = Time()
  n46 = ScalarParameter() {defaultValue=0.7, parameterName=Speed}
  n47 = TextureSampleParameter2D(UVs=Panner(Coordinate=Add(A=@n43, B=@n44), Time=Multiply(A=@n45, B=@n46)) {bFractionalPart=true, speedX=-1, speedY=0.05}) {parameterName=T_FlamesTex, samplerType=SAMPLERTYPE_LinearColor, texture=&T_Plasma_Flames_02_M}
  n48 = ScalarParameter() {parameterName=Flames_Distortion}
  n49 = Multiply(A=ScalarParameter() {parameterName=Layer2Size}, B=Add(A=@n43, B=NamedRerouteUsage()))
  n50 = Multiply(A=Time(), B=ScalarParameter() {defaultValue=0.7, parameterName=Speed})
  n51 = TextureSampleParameter2D(UVs=Add(A=Multiply(A=@n48, B=@n47.R), B=Panner(Coordinate=@n49, Time=@n50) {bFractionalPart=true, speedX=-1, speedY=-0.05})) {parameterName=T_FlamesTex, samplerType=SAMPLERTYPE_LinearColor, texture=&T_Plasma_Flames_02_M}
  n52 = NamedRerouteDeclaration(None=Multiply(A=Add(A=@n47.R, B=@n51.R)) {constB=0.5}) {name=MainNoise, nodeColor=<rgba 0 0.1328 1 1>}
  n53 = Multiply(A=Multiply(A=Constant() {r=1}, B=SmoothStep(Min=@n38, Max=@n39, Value=@n52)), B=VertexColor().G)
  n54 = DepthFade(Opacity=@n53) {fadeDistanceDefault=4}
  QualitySwitch(Default=@n53, Low=@n53, High=@n54, Medium=@n53, Epic=@n54)
  ScalarParameter() {defaultValue=1, group=Overlay, parameterName=OverlayTile}
  TextureCoordinate() {uTiling=0.125, vTiling=0.5}
  TextureCoordinate() {uTiling=0.2, vTiling=0.478}
  VertexNormalWS()
 ]
}
