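# Unlit material for the glowing, churning innards effect on the body of the female BariumDemon character.
# Two flow-distorted noise layers scroll over planar coordinates from pre-skinned position normalized by mesh bounds, keeping them attached during animation.
# Noise, a distorted insides texture, spherical internal light, overlay detail and smoothstepped highlights build an intensity mask that screen-blends bright color over base color before contrast and emissive adjustment.
# A masked camera-offset branch provides world position offset.
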Material /BRCosmetics/Characters/Player/Female/Medium/Bodies/F_MED_BariumDemon/FX/M_MED_BariumDemon_FX_BodyInnards
graph {
 properties {bUsedWithMeshDeformer=true, bUsedWithSkeletalMesh=true, pixelDepthOffsetMode=PDOM_Legacy, shadingModel=MSM_Unlit, usedShadingModels=Unlit}
 nodes [
  n1 = VectorParameter() {group=Colors, parameterName=BrightColor, defaultValue=<rgba 0 0 0 1>}
  n2 = VertexInterpolator(VS=Divide(A=PreSkinnedPosition(), B=PreSkinnedLocalBounds().Half Extents))
  n3 = ComponentMask(None=@n2.PS) {b=true, r=true}
  n4 = VectorParameter() {parameterName=InsidesScale_1, channelNames {a=Speed_Y, b=Speed_X, g=Scale_Y, r=Scale_X}, defaultValue=<rgba 2.5 0.7 0 1>}
  n5 = Multiply(A=@n3, B=ComponentMask(None=@n4) {g=true, r=true})
  n6 = Panner(Coordinate=Multiply(A=TextureCoordinate(), B=Constant2Vector() {g=2, r=2}) {constB=2}) {speedY=-0.25}
  n7 = Multiply(
   A=ComponentMask(None=TextureSample(UVs=@n6) {samplerType=SAMPLERTYPE_LinearColor, texture=/Game/Global/Textures/Default/Flowmaps/T_SwirlFlow_01}.RGB) {g=true, r=true}
   B=ScalarParameter() {group=FlowMaps, parameterName=FlowStr}
  )
  n8 = Multiply(A=NamedRerouteUsage(), B=ScalarParameter() {defaultValue=1, parameterName=InsidesSpeed})
  n9 = TextureSampleParameter2D(UVs=Panner(Coordinate=Add(A=@n5, B=@n7), Time=@n8, Speed=AppendVector(A=@n4.Speed_X, B=@n4.Speed_X)) {bFractionalPart=true, speedY=-0.35}) {group=Textures, parameterName=T_MainNoise, samplerType=SAMPLERTYPE_LinearColor, texture=/Game/Items/Traps/Textures/T_Plasma_Flames_02_M}
  n10 = ScalarParameter() {parameterName=Noise_distor}
  n11 = VectorParameter() {parameterName=InsidesScale_2, channelNames {a=Speed_Y, b=Speed_X, g=Scale_Y, r=Scale_X}, defaultValue=<rgba 2.5 0.7 0 1>}
  n12 = Add(A=Multiply(A=@n3, B=ComponentMask(None=@n11) {g=true, r=true}), B=@n7)
  n13 = AppendVector(A=@n11.Speed_X, B=@n11.Speed_X)
  n14 = TextureSampleParameter2D(UVs=Add(A=Multiply(A=@n9.G, B=@n10) {constB=0.1}, B=Panner(Coordinate=@n12, Time=@n8, Speed=@n13) {bFractionalPart=true, speedX=0.2, speedY=-0.8})) {group=Textures, parameterName=T_MainNoise, samplerType=SAMPLERTYPE_LinearColor, texture=&T_Plasma_Flames_02_M}
  n15 = Multiply(A=Add(A=@n9.G, B=@n14.G)) {constB=0.75}
  n16 = Add(A=NamedRerouteUsage(), B=Multiply(A=NamedRerouteUsage(), B=@n15))
  n17 = Subtract(A=Multiply(A=NamedRerouteUsage()) {constB=2})
  n18 = ScalarParameter() {parameterName=Insides_Tex_Distortion}
  n19 = TextureSample(UVs=Add(A=TextureCoordinate() {uTiling=2}, B=Multiply(A=@n17, B=@n18))) {texture=&/T_FX_BariumDemon_Insides}
  n20 = Multiply(A=@n19.R, B=ScalarParameter() {parameterName=Insides_Strength})
  n21 = NamedRerouteUsage()
  n22 = ScalarParameter() {parameterName=highlights_str}
  n23 = Multiply(A=Add(A=Add(A=@n16, B=@n20), B=Multiply(A=@n21, B=@n22)), B=@n19.G)
  n24 = MaterialFunctionCall!(
   Color=MaterialFunctionCall!(Blend=Multiply(A=@n1.RGB, B=@n23), Base=VectorParameter() {group=Colors, parameterName=BaseColor, defaultValue=<rgba 0 0 0 1>}.RGB) {materialFunction=/Engine/Functions/Engine_MaterialFunctions03/Blends/Blend_Screen}.Result
   Contrast=ScalarParameter() {parameterName=Contrast}
  ) {materialFunction=/Engine/Functions/Engine_MaterialFunctions01/ImageAdjustment/Contrast_Preserve_Color}
  Add(A=NamedRerouteUsage(), B=@n24.Result)
  MaterialFunctionCall(Emissive=@n24.Result) {materialFunction=&MF_FX_EmissiveAdjustments}
  Multiply(
   A=MaterialFunctionCall(Offset Amount=ScalarParameter() {group=WPO, parameterName=CameraOffset}) {materialFunction=/Engine/Functions/Engine_MaterialFunctions02/WorldPositionOffset/CameraOffset}.Camera Offset
   B=NamedRerouteUsage()
  )
  NamedRerouteDeclaration(None=Multiply(A=TextureCoordinate(), B=ScalarParameter() {defaultValue=1, group=Insides, parameterName=InsidesTile})) {name=InsidesUV, nodeColor=<rgba 0 0.0156 1 1>}
  n25 = VectorParameter() {parameterName=InternalLightPos, channelNames {a=MaxRadius, b=MinRadius, g="", r=""}, defaultValue=<rgba 0.5 0.5 0 1>}
  n26 = TextureSampleParameter2D(UVs=Panner(Coordinate=Constant2Vector()) {speedX=0.1}) {group=Textures, parameterName=T_MainNoise, samplerType=SAMPLERTYPE_LinearColor, texture=&T_Plasma_Flames_02_M}
  n27 = SphereMask(
   A=ComponentMask(None=@n2.PS) {b=true, r=true}
   B=ComponentMask(None=@n25) {g=true, r=true}
   Radius=LinearInterpolate(A=@n25.MinRadius, B=@n25.MinRadius, Alpha=VertexInterpolator(VS=@n26.G).PS)
  ) {attenuationRadius=0.5, hardnessPercent=0}
  NamedRerouteDeclaration(None=Multiply(A=Power(Base=@n27) {constExponent=4}, B=ScalarParameter() {defaultValue=2, parameterName=Internal_light_brightness})) {name=InternalLight, nodeColor=<rgba 1 0 0.1641 1>}
  NamedRerouteDeclaration(None=@n15) {name=MainNoise, nodeColor=<rgba 1 0.7031 0 1>}
  n28 = TextureCoordinate()
  n29 = ScalarParameter() {group=Overlay, parameterName=overlay_tile}
  n30 = NamedRerouteUsage()
  n31 = ScalarParameter() {group=Overlay, parameterName=OverlayDistortions}
  n32 = Panner(Coordinate=Add(A=Multiply(A=@n28, B=@n29), B=Multiply(A=@n30, B=@n31)), Time=NamedRerouteUsage()) {speedY=-1.2}
  NamedRerouteDeclaration(None=Power(
    Base=TextureSampleParameter2D(UVs=@n32) {parameterName=T_OverlayDetails, samplerType=SAMPLERTYPE_LinearColor, texture=&T_Plasma_Flames_02_M}.R
    Exp=ScalarParameter() {group=Overlay, parameterName=OverlayContrast}
   )) {name=OverlayDetails, nodeColor=<rgba 0 1 0.9063 1>}
  NamedRerouteDeclaration(None=@n20) {name=Insides, nodeColor=<rgba 1 0 0.8437 1>}
  NamedRerouteDeclaration(None=SmoothStep(Min=ScalarParameter() {parameterName=brights_min}, Max=ScalarParameter() {parameterName=brights_max}, Value=NamedRerouteUsage())) {name=Highlights, nodeColor=<rgba 0.2656 1 0 1>}
  NamedRerouteDeclaration(None=Multiply(A=Time(), B=ScalarParameter() {defaultValue=1, group=Global, parameterName=EffectTiming})) {name=FX_Timing, nodeColor=<rgba 0 1 0.7188 1>}
  NamedRerouteDeclaration(None=VertexColor().R) {name=CameraOffset_Mask, nodeColor=<rgba 0.6641 0 1 1>}
  NamedRerouteUsage()
  NamedRerouteUsage()
  NamedRerouteUsage()
  n33 = Subtract(A=TextureCoordinate(), B=Constant2Vector() {r=0.5})
  n34 = VectorParameter() {parameterName=shoulders_cam_offset_mask_xfm, defaultValue=<rgba 0.289 0.268 0.136 1>}
  OneMinus(None=SphereMask(A=Abs(None=@n33), B=ComponentMask(None=@n34.RGB) {g=true, r=true}, Radius=@n34.B) {attenuationRadius=0.175, hardnessPercent=50})
 ]
}
