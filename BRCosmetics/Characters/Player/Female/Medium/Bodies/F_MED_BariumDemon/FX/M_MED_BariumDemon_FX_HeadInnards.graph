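# Unlit skeletal-mesh material for the glowing, animated head innards of the female Barium Demon character skin.
# Radial UVs drive two panning plasma-noise samples, the second distorted by the first, plus a tiled, distorted plasma overlay.
# A sphere mask around an internal light position adds a noise-modulated glow, which combines with the overlay, tints by bright color and screen-blends over base color into emissive, alongside a camera offset.
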Material /BRCosmetics/Characters/Player/Female/Medium/Bodies/F_MED_BariumDemon/FX/M_MED_BariumDemon_FX_HeadInnards
graph {
 properties {bUsedWithMeshDeformer=true, bUsedWithSkeletalMesh=true, pixelDepthOffsetMode=PDOM_Legacy, shadingModel=MSM_Unlit, usedShadingModels=Unlit}
 nodes [
  Add()
  n1 = VertexInterpolator(VS=Divide(A=PreSkinnedPosition(), B=PreSkinnedLocalBounds().Half Extents))
  ComponentMask(None=@n1.PS) {b=true, r=true}
  MaterialFunctionCall(Offset Amount=ScalarParameter() {parameterName=CamOff}) {materialFunction=/Engine/Functions/Engine_MaterialFunctions02/WorldPositionOffset/CameraOffset}
  n2 = VectorParameter() {parameterName=InternalLightPos, channelNames {a=MaxRadius, b=MinRadius, g="", r=""}, defaultValue=<rgba 0.5 0.5 0 1>}
  n3 = Multiply(A=NamedRerouteUsage()) {constB=0.25}
  n4 = NamedRerouteUsage()
  n5 = NamedRerouteUsage()
  n6 = SphereMask(A=TextureCoordinate(), B=ComponentMask(None=@n2) {g=true, r=true}, Radius=Subtract(A=Add(A=@n3, B=@n4), B=Multiply(A=@n5) {constB=3})) {attenuationRadius=0.5, hardnessPercent=0}
  n7 = ScalarParameter() {defaultValue=1, parameterName=InternalLightBrightness}
  n8 = Constant() {r=0.5}
  n9 = Add(A=Multiply(A=NamedRerouteUsage(), B=@n8), B=@n8)
  n10 = Add(
   A=Multiply(A=Power(Base=@n6) {constExponent=4}, B=Multiply(A=@n7, B=@n9))
   B=Multiply(A=NamedRerouteUsage(), B=ScalarParameter() {group=Overlay, parameterName=Overlay_Brightness})
  )
  n11 = VectorParameter() {group=Colors, parameterName=Bright_Color, defaultValue=<rgba 0 0 0 1>}
  MaterialFunctionCall(Emissive=MaterialFunctionCall!(Blend=Multiply(A=@n10, B=@n11.RGB), Base=VectorParameter() {group=Colors, parameterName=Base_Color, defaultValue=<rgba 0 0 0 1>}.RGB) {materialFunction=/Engine/Functions/Engine_MaterialFunctions03/Blends/Blend_Screen}.Result) {materialFunction=&MF_FX_EmissiveAdjustments}
  n12 = Subtract(A=TextureCoordinate(), B=ComponentMask(None=VectorParameter() {parameterName=RadialSrc, defaultValue=<rgba 0 0 0 1>}.RGB) {g=true, r=true})
  NamedRerouteDeclaration(None=Multiply(
    A=MaterialFunctionCall(Vector or UVs=@n12) {materialFunction=&VectorToRadialValue}.Radial Coordinates
    B=ScalarParameter() {defaultValue=1, group=Insides, parameterName=InsidesTile}
   )) {name=InsidesUV, nodeColor=<rgba 0 0.0156 1 1>}
  n13 = NamedRerouteUsage()
  n14 = VectorParameter() {parameterName=UVScale, defaultValue=<rgba 8 0.6 0 1>}
  n15 = ComponentMask(None=@n14.RGB) {g=true, r=true}
  n16 = Multiply(A=@n13, B=@n15)
  n17 = Multiply(A=Constant2Vector(), B=ScalarParameter() {group=FlowMaps, parameterName=FlowStr})
  n18 = Multiply(A=NamedRerouteUsage(), B=ScalarParameter() {defaultValue=1, parameterName=InsidesSpeed})
  n19 = TextureSampleParameter2D(UVs=Panner(Coordinate=Add(A=@n16, B=@n17), Time=@n18) {bFractionalPart=true, speedY=-1}) {group=Textures, parameterName=T_MainNoise, samplerType=SAMPLERTYPE_LinearColor, texture=/Game/Items/Traps/Textures/T_Plasma_Flames_02_M}
  n20 = ScalarParameter() {parameterName=Noise_distor}
  n21 = Add(A=Multiply(A=@n13, B=Multiply(A=@n15, B=@n14.B)), B=@n17)
  n22 = TextureSampleParameter2D(UVs=Add(A=Multiply(A=@n19.R, B=@n20) {constB=0.1}, B=Panner(Coordinate=@n21, Time=@n18) {bFractionalPart=true, speedX=1})) {group=Textures, parameterName=T_MainNoise, samplerType=SAMPLERTYPE_LinearColor, texture=&T_Plasma_Flames_02_M}
  NamedRerouteDeclaration(None=Multiply(A=Add(A=@n19.R, B=@n22.R))) {name=MainNoise, nodeColor=<rgba 1 0.7031 0 1>}
  NamedRerouteDeclaration(None=ComponentMask(None=@n1.PS) {b=true, g=true}) {name=PanningUV, nodeColor=<rgba 0 1 0.25 1>}
  NamedRerouteDeclaration(None=SmoothStep(Min=ScalarParameter() {parameterName=brights_min}, Max=ScalarParameter() {parameterName=brights_max}, Value=NamedRerouteUsage())) {name=Highlights, nodeColor=<rgba 0.2656 1 0 1>}
  NamedRerouteDeclaration(None=Multiply(A=Time(), B=ScalarParameter() {defaultValue=1, group=Global, parameterName=EffectTiming})) {name=FX_Timing, nodeColor=<rgba 0 1 0.7188 1>}
  n23 = Noise!(World Position=Time(), FilterWidth) {levels=1, noiseFunction=NOISEFUNCTION_GradientTex3D, outputMin=0}
  NamedRerouteDeclaration(None=LinearInterpolate(A=@n2.MinRadius, B=@n2.MinRadius, Alpha=VertexInterpolator(VS=@n23).PS)) {name=LightRadius, nodeColor=<rgba 0.875 1 0 1>}
  n24 = VectorParameter() {group=Overlay, parameterName=Overlay_Tiling, defaultValue=<rgba 2 1 0 1>}
  n25 = Multiply(A=ComponentMask(None=@n24.RGB) {g=true, r=true}, B=@n24.B)
  n26 = NamedRerouteUsage()
  n27 = Subtract(A=Multiply(A=NamedRerouteUsage()) {constB=2})
  n28 = ScalarParameter() {group=Overlay, parameterName=Overlay_Distortion}
  n29 = Panner(Coordinate=Add(A=Multiply(A=@n25, B=@n26), B=Multiply(A=@n27, B=@n28))) {bFractionalPart=true, speedY=-0.25}
  NamedRerouteDeclaration(None=Power(
    Base=TextureSample(UVs=@n29) {samplerType=SAMPLERTYPE_LinearGrayscale, texture=/Game/Effects/Fort_Effects/Emotes/UltraLight/Texture/T_Plasma_C_Greyscale}.R
    Exp=ScalarParameter() {group=Overlay, parameterName=Overlay_Contrast}
   )) {name=OverlayDetails, nodeColor=<rgba 0 0.0156 1 1>}
 ]
}
