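# Niagara idle effect for the BariumDemon character's chest: a flickering green energy glow with small wisps of smoke.
# InternalEnergy holds a looping burst glow sprite whose waveform-driven intensity scales both color and light renderer, while system-age curves animate material parameters.
# InternalEnergyCore adds a smaller, brighter sprite and mesh core pulsed by its own waveform.
# Smoke_Small continuously spawns rotating SubUV sprites that drift down, swell and fade.
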NiagaraSystem /BRCosmetics/Characters/Player/Female/Medium/Bodies/F_MED_BariumDemon/FX/NS_F_MED_BariumDemon_Chest_Idle
dependencies {
 usedDataInterfaces [&NiagaraDataInterfaceCurve, &NiagaraDataInterfaceColorCurve]
 usedDynamicInputs []
 usedModules [&SystemState]
}
emitters {
 InternalEnergy {
  emitterUpdateScript [
   &EmitterState(
    Inactive Response=::NewEnumerator1
    Life Cycle Mode=::NewEnumerator1
    Loop Behavior=::NewEnumerator1
    Loop Duration Mode=::NewEnumerator1
    UseLoopDelay=false
   )
   &SpawnBurst_Instantaneous(Use Spawn Probability=false, UseLoopCountLimit=false)
  ]
  particleSpawnScript [&InitializeParticle(
    Color=<rgba 1.0238 2 0.3665 1>
    Lifetime=$Emitter.CurrentLoopDuration
    Position Offset=<-2 2 0>
    Sprite Size Mode=::NewEnumerator3
    Uniform Sprite Size=40
    UsePositionOffset=true
    Write Material Random=true
   )]
  particleUpdateScript [
   &ParticleState(
    Kill Particles When Lifetime Has Elapsed=false
    Let Infinitely Lived Particles Die When Emitter Deactivates=true
    Loop Particles Lifetime=true
   )
   &SolveForcesAndVelocity(Clamp Velocity=false, Limit Acceleration=false)
   &SpriteFacingAndAlignment(
    Alignment Coordinate Space=::Local
    Facing Coordinate Space=::Local
    Sprite Alignment=<0 0 1>
    Sprite Facing=<0 1 0>
    WriteAlignment=true
    WriteFacing=true
   )
   /SetVariables(Particles.Intensity:NiagaraFloat=&Waveform(
     Waveform Count=::NewEnumerator1
     "Waveform [1]"=::NewEnumerator12
     "[1] Frequency"=5
     "[1] Amplitude Min/Max"=<0 1>
     "[2] Average"=false
     "Waveform [2]"=::NewEnumerator12
     "[2] Input Position"=0
     "[2] Phase"=$Emitter.Age
     "[2] Phase Scale"=1
     "[2] Compound Frequency"=2
     "[2] Frequency"=7
     "[2] Amplitude Min/Max"=<0 1>
     "[2] Scale"=1
     "[2] Bias"=0
     "[2] Exponent"=1
     "[2] Pulse Duration"=0.1
     "[2] Clip Triangle Min/Max"=<0 0>
     "[2] Random Seed"=0
    )) {isSetParameters=true}
   /SetVariables(Particles.GlowSize=<45 70>) {isSetParameters=true}
   /SetVariables(Particles.SpriteFacing=&TransformVector(Vector=<0 1 0>, Source Space=::Simulation, Destination Space=::Local)) {isSetParameters=true, module=SetVariables_#1}
   &ScaleColor(
    Scale Alpha=0.05
    Scale RGB=&VectorFromFloat(Value=&Multiply_Float(A=$Particles.Intensity, B=1))
    ScaleA=true
    ScaleRGB=true
   )
   /SetVariables(Particles.GlowParams:Vector4f=<0 2 0 0>) {isSetParameters=true}
   &DynamicMaterialParameters(
    Index 0 Param 1=&Lerp_Float(
     A=0.3
     B=0.6
     Alpha=&FloatFromCurve(
      FloatCurve=DataInterface() {propertyValues {Curve {keys [
          <at 0 1 RCIM_Cubic>
          {arriveTangent=0.625, interpMode=RCIM_Cubic, leaveTangent=0.625, time=0.1, value=0.75}
          {arriveTangent=-0.8333, interpMode=RCIM_Cubic, leaveTangent=-0.8333, time=0.2, value=1.125}
          {arriveTangent=0.4167, interpMode=RCIM_Cubic, leaveTangent=0.4167, time=0.4, value=0.5}
          {arriveTangent=1.25, interpMode=RCIM_Cubic, leaveTangent=1.25, time=0.5, value=1.25}
          {arriveTangent=-0.625, interpMode=RCIM_Cubic, leaveTangent=-0.625, time=0.6, value=0.75}
          {arriveTangent=0.625, interpMode=RCIM_Cubic, leaveTangent=0.625, time=0.7, value=1.125}
          <at 1 1 RCIM_Cubic>
         ]}}}
      CurveIndex=&ModuloFloat(A=&Multiply_Float(A=$System.Age, B=0.75), B=1)
     )
    )
    Index 0 Param 2=&Lerp_Float(
     A=0
     B=3
     Alpha=&FloatFromCurve(
      FloatCurve=DataInterface() {propertyValues {Curve {keys [
          {arriveTangent=0.625, interpMode=RCIM_Cubic, time=0, value=0.5}
          {arriveTangent=-1.75, interpMode=RCIM_Cubic, leaveTangent=-1.75, time=0.1, value=1.1}
          {arriveTangent=1.375, interpMode=RCIM_Cubic, leaveTangent=1.375, time=0.2, value=0.15}
          {arriveTangent=-0.125, interpMode=RCIM_Cubic, leaveTangent=-0.125, time=0.3, value=1.375}
          {arriveTangent=-4.375, interpMode=RCIM_Cubic, leaveTangent=-4.375, time=0.4, value=0.125}
          <at 0.5 0.5 RCIM_Cubic>
          {arriveTangent=2.25, interpMode=RCIM_Cubic, leaveTangent=2.25, time=0.7, value=0.125}
          {arriveTangent=1.25, interpMode=RCIM_Cubic, leaveTangent=1.25, time=0.8, value=1.175}
          <at 1 0.5 RCIM_Cubic>
         ]}}}
      CurveIndex=&ModuloFloat(A=&Multiply_Float(A=$System.Age, B=1), B=1)
     )
    )
    Module.Param0WriteEnabled=true
    Module.Param1WriteEnabled=true
    Module.Param2WriteEnabled=true
    Module.Param3WriteEnabled=true
   )
   /SetVariables() {isSetParameters=true}
   &FN_Gen5_Light(Attenutation=1, Intensity:NiagaraFloat=&Multiply_Float(A=$Particles.Intensity, B=10))
   /SetVariables() {isSetParameters=true}
   /SetVariables(Particles.LightColor=&MultiplyLinearColor(
     Linear Color=$Particles.Initial.Color
     Scale Factor=&MakeVector4_FromFloat(Float=$StackContext.FN_Gen5_Light.Intensity)
    )) {isSetParameters=true}
   /SetVariables(Particles.LightColor=&ScaleLinearColorByCurve(
     InitialValue=$Particles.LightColor
     ScaleCurve=DataInterface() {propertyValues {AlphaCurve {keys []}, BlueCurve {keys []}, GreenCurve {keys []}, RedCurve {keys []}}}
     CurveIndex=$Particles.NormalizedAge
    )) {enabled=false, isSetParameters=true, module=SetVariables_#2}
  ]
  rendererClasses [&NiagaraSpriteRendererProperties, &NiagaraLightRendererProperties]
  renderers [
   {rendererClass=&NiagaraSpriteRendererProperties}
   {rendererClass=&NiagaraLightRendererProperties, rendererIndex=1}
   {rendererClass=&NiagaraSpriteRendererProperties, rendererIndex=2}
  ]
 }
 InternalEnergyCore {
  emitterUpdateScript [
   &EmitterState(
    Inactive Response=::NewEnumerator1
    Life Cycle Mode=::NewEnumerator1
    Loop Behavior=::NewEnumerator1
    Loop Duration Mode=::NewEnumerator1
    UseLoopDelay=false
   )
   &SpawnBurst_Instantaneous(Use Spawn Probability=false, UseLoopCountLimit=false)
  ]
  particleSpawnScript [&InitializeParticle(
    Color=<rgba 17.6562 30 5.3125 1>
    Lifetime=$Emitter.CurrentLoopDuration
    Position Offset=<3 2.5 0>
    Sprite Size Mode=::NewEnumerator3
    Uniform Sprite Size=16.0755
    UsePositionOffset=true
    Write Material Random=true
   )]
  particleUpdateScript [
   &ParticleState(
    Kill Particles When Lifetime Has Elapsed=false
    Let Infinitely Lived Particles Die When Emitter Deactivates=true
    Loop Particles Lifetime=true
   )
   &SolveForcesAndVelocity(Clamp Velocity=false, Limit Acceleration=false)
   &SpriteFacingAndAlignment(
    Alignment Coordinate Space=::Local
    Facing Coordinate Space=::Local
    Sprite Alignment=<0 0 1>
    Sprite Facing=<0 1 0>
    WriteAlignment=true
    WriteFacing=true
   )
   /SetVariables(Particles.SpriteFacing=&TransformVector(Vector=<0 1 0>, Source Space=::Simulation, Destination Space=::Local)) {isSetParameters=true, module=SetVariables_#1}
   &ScaleColor(
    Scale Alpha=0.2
    Scale RGB=&VectorFromFloat(Value=&Waveform(
      Waveform Count=::NewEnumerator1
      "Waveform [1]"=::NewEnumerator12
      "[1] Frequency"=1
      "[1] Amplitude Min/Max"=<0 1>
      "[2] Average"=false
      "Waveform [2]"=::NewEnumerator12
      "[2] Input Position"=0
      "[2] Phase"=$Emitter.Age
      "[2] Phase Scale"=2
      "[2] Compound Frequency"=2
      "[2] Frequency"=2
      "[2] Amplitude Min/Max"=<0 1>
      "[2] Scale"=2
      "[2] Bias"=0
      "[2] Exponent"=1
      "[2] Pulse Duration"=0.1
      "[2] Clip Triangle Min/Max"=<0 0>
      "[2] Random Seed"=0
     ))
    ScaleA=true
    ScaleRGB=true
   )
   &DynamicMaterialParameters(
    Index 0 Param 1=&Lerp_Float(
     A=0.2
     B=0.6
     Alpha=&FloatFromCurve(
      FloatCurve=DataInterface() {propertyValues {Curve {keys [
          <at 0 1 RCIM_Cubic>
          {arriveTangent=0.625, interpMode=RCIM_Cubic, leaveTangent=0.625, time=0.1, value=0.75}
          {arriveTangent=-0.8333, interpMode=RCIM_Cubic, leaveTangent=-0.8333, time=0.2, value=1.125}
          {arriveTangent=0.4167, interpMode=RCIM_Cubic, leaveTangent=0.4167, time=0.4, value=0.5}
          {arriveTangent=1.25, interpMode=RCIM_Cubic, leaveTangent=1.25, time=0.5, value=1.25}
          {arriveTangent=-0.625, interpMode=RCIM_Cubic, leaveTangent=-0.625, time=0.6, value=0.75}
          {arriveTangent=0.625, interpMode=RCIM_Cubic, leaveTangent=0.625, time=0.7, value=1.125}
          <at 1 1 RCIM_Cubic>
         ]}}}
      CurveIndex=&ModuloFloat(A=&Multiply_Float(A=$System.Age, B=0.75), B=1)
     )
    )
    Index 0 Param 2=&Lerp_Float(
     A=0
     B=1
     Alpha=&FloatFromCurve(
      FloatCurve=DataInterface() {propertyValues {Curve {keys [
          {arriveTangent=0.625, interpMode=RCIM_Cubic, time=0, value=0.5}
          {arriveTangent=-1.75, interpMode=RCIM_Cubic, leaveTangent=-1.75, time=0.1, value=1.1}
          {arriveTangent=1.375, interpMode=RCIM_Cubic, leaveTangent=1.375, time=0.2, value=0.15}
          {arriveTangent=-0.125, interpMode=RCIM_Cubic, leaveTangent=-0.125, time=0.3, value=1.375}
          {arriveTangent=-4.375, interpMode=RCIM_Cubic, leaveTangent=-4.375, time=0.4, value=0.125}
          <at 0.5 0.5 RCIM_Cubic>
          {arriveTangent=2.25, interpMode=RCIM_Cubic, leaveTangent=2.25, time=0.7, value=0.125}
          {arriveTangent=1.25, interpMode=RCIM_Cubic, leaveTangent=1.25, time=0.8, value=1.175}
          <at 1 0.5 RCIM_Cubic>
         ]}}}
      CurveIndex=&ModuloFloat(A=&Multiply_Float(A=$System.Age, B=1), B=1)
     )
    )
    Module.Param0WriteEnabled=true
    Module.Param1WriteEnabled=true
    Module.Param2WriteEnabled=true
    Module.Param3WriteEnabled=true
   ) {enabled=false}
   /SetVariables(Particles.LightColor=&ScaleLinearColorByCurve(
     InitialValue=$Particles.LightColor
     ScaleCurve=DataInterface() {propertyValues {AlphaCurve {keys []}, BlueCurve {keys []}, GreenCurve {keys []}, RedCurve {keys []}}}
     CurveIndex=$Particles.NormalizedAge
    )) {enabled=false, isSetParameters=true, module=SetVariables_#2}
  ]
  rendererClasses [&NiagaraSpriteRendererProperties, &NiagaraMeshRendererProperties]
  renderers [
   {rendererClass=&NiagaraSpriteRendererProperties}
   {rendererClass=&NiagaraMeshRendererProperties, rendererIndex=1}
   {rendererClass=&NiagaraSpriteRendererProperties, rendererIndex=2}
  ]
 }
 Smoke_Small {
  emitterUpdateScript [
   &EmitterState(Life Cycle Mode=::NewEnumerator1, Loop Duration=1, UseLoopDelay=false)
   &SpawnRate(SpawnRate=5, Use Spawn Probability=false)
  ]
  particleSpawnScript [
   &InitializeParticle(
    Color=<rgba 0.0652 0.125 0.0078 1>
    Lifetime=&RandomRangeFloat(Minimum=1, Maximum=1.75)
    Position Offset=<1 2 2>
    Sprite Rotation Mode=::NewEnumerator2
    Sprite Size Mode=::NewEnumerator1
    Uniform Sprite Size Max=20
    Uniform Sprite Size Min=10
    UsePositionOffset=true
    Write Material Random=true
   ) {module=InitializeParticle001}
   &SubUVAnimation(End Frame=4, SubUV Animation Mode:ENiagara_SubUVLookupMode=::NewEnumerator2)
   &ShapeLocation(MaskSpawnGroup=false, Override Seed=false, Sphere Radius=5, UseOwnerScale=true)
  ]
  particleUpdateScript [
   &ParticleState()
   &Drag(Drag=6, Use Linear Drag=true, Use Rotational Drag=false)
   &GravityForce(Gravity=<0 0 -30.5917>)
   &SolveForcesAndVelocity(Clamp Velocity=false, Limit Acceleration=false)
   &DynamicMaterialParameters(
    Index 0 Param 2=1
    Index 0 Param 3=2
    Index 0 Param 4=&FloatFromCurve(FloatCurve=DataInterface() {propertyValues {Curve {keys [<at 0 1>,<at 1 0.22>]}}})
    Module.Param0WriteEnabled=false
    Module.Param1WriteEnabled=true
    Module.Param2WriteEnabled=true
    Module.Param3WriteEnabled=true
   )
   &SpriteRotationRate(Rotation Rate=&Multiply_Float(
     A=100
     B=&FloatFromCurve(
      FloatCurve=DataInterface() {propertyValues {Curve {keys [<at 0 1>,<at 1 0>]}}}
      Scale Curve=&RandomRangeFloat(Minimum=-1, Maximum=1)
     )
    ))
   &ScaleColor(
    Scale Alpha=&Multiply_Float(
     A=&FloatFromCurve(FloatCurve=DataInterface() {propertyValues {Curve {keys [
          {interpMode=RCIM_Cubic, tangentMode=RCTM_User, time=0, value=0}
          {arriveTangent=-3.2207, interpMode=RCIM_Cubic, leaveTangent=-3.2207, tangentMode=RCTM_User, time=0.1, value=1.275}
          {tangentMode=RCTM_User, time=1, value=0}
         ]}}})
     B=1
    )
    Scale RGB=&VectorFromFloat(Value=&FloatFromCurve(
      FloatCurve=DataInterface() {propertyValues {Curve {keys [<at 0 0 RCIM_Cubic>, {arriveTangent=0.25, interpMode=RCIM_Cubic, leaveTangent=0.25, time=0.5, value=2.375}, <at 1 0.25 RCIM_Cubic>]}}}
      Scale Curve=10
     ))
    ScaleA=true
    ScaleRGB=true
   )
   &ScaleSpriteSize(Scale Factor=&Vector2DFromFloat(Value=&FloatFromCurve(FloatCurve=DataInterface() {propertyValues {Curve {keys [
          <at 0 0.8 RCIM_Cubic>
          {arriveTangent=-0.4771, interpMode=RCIM_Cubic, leaveTangent=-0.4771, tangentMode=RCTM_User, time=0.7, value=1.25}
          <at 1 1.2 RCIM_Cubic>
         ]}}})))
  ]
  rendererClasses [&NiagaraSpriteRendererProperties]
  renderers [{rendererClass=&NiagaraSpriteRendererProperties}]
 }
}
summary {
 systemName=NS_F_MED_BariumDemon_Chest_Idle
 emitters [
  {emitterName=Smoke_Small, rendererClasses [&NiagaraSpriteRendererProperties]}
  {
   emitterName=InternalEnergy
   rendererClasses [&NiagaraSpriteRendererProperties, &NiagaraLightRendererProperties]
  }
  {
   emitterName=InternalEnergyCore
   rendererClasses [&NiagaraSpriteRendererProperties, &NiagaraMeshRendererProperties]
  }
 ]
 userVariables []
}
