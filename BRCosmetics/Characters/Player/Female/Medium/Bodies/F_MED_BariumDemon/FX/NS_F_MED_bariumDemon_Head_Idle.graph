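# Niagara idle effect that gives the female BariumDemon player cosmetic a green flaming head.
# Flames continuously spawns short-lived animated sprites, while two mesh emitters hold looping pairs: one mirrored on X with waveform-driven material parameters, one separately offset and oriented.
# A Light emitter flickers through waveform intensity, wanders via smoothed random offsets and scales its color by User.FrontEnd.
# User.bisHeadHidden sets visibility tags on flames and meshes and disables the light.
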NiagaraSystem /BRCosmetics/Characters/Player/Female/Medium/Bodies/F_MED_BariumDemon/FX/NS_F_MED_bariumDemon_Head_Idle
dependencies {
 usedDataInterfaces [
  &NiagaraDataInterfaceArrayFloat
  &NiagaraDataInterfaceArrayFloat3
  &NiagaraDataInterfaceCurve
  &NiagaraDataInterfaceColorCurve
  &NiagaraDataInterfaceVectorCurve
 ]
 usedDynamicInputs [&MakeVector4_FromLinearColor, &ColorFromCurve]
 usedModules [&SystemState]
}
emitters {
 Flames {
  emitterUpdateScript [
   &EmitterState(Life Cycle Mode=::NewEnumerator1, Loop Duration=1, UseLoopDelay=false)
   &SpawnRate(SpawnRate=&UniformRangedFloat(Minimum=5, Maximum=10), Use Spawn Probability=false)
  ]
  particleSpawnScript [
   /Niagara/Modules/Spawn/Initialization/InitializeParticle(
    Color=&UniformRangedLinearColor(Minimum=<rgba 0.8 0.8 0.8 0.35>, Maximum=<rgba 1 1 1 0.5>)
    Lifetime=&UniformRangedFloat(Minimum=0.5, Maximum=1)
    Module.Write Color=true
    Module.Write Lifetime=true
    Module.Write Mass=true
    Module.Write Material Random=true
    Module.Write Position=false
    Module.Write Scale=false
    Module.Write SpriteRotation=true
    Module.Write SpriteSize=true
    Position=&SimulationPosition()
    Sprite Rotation=&UniformRangedFloat(Minimum=0, Maximum=360)
    Sprite Size=&Vector2DFromFloat(Value=&UniformRangedFloat(Minimum=30, Maximum=40))
   ) {enabled=false}
   /Niagara/Modules/Spawn/Initialization/V2/InitializeParticle(
    Color=<rgba 0.2606 0.5 0.0312 1>
    Lifetime=&RandomRangeFloat(Minimum=0.5, Maximum=1.2)
    Position Offset=<0 0 5>
    Sprite Rotation Mode=::NewEnumerator2
    Sprite Size Mode=::NewEnumerator3
    Uniform Sprite Size=25
    UsePositionOffset=true
    Write Material Random=true
   ) {module=InitializeParticle001}
   &ApplyInitialForces()
   &DynamicMaterialParameters(
    Index 0 Param 2=&UniformRangedFloat(Minimum=0, Maximum=1)
    Index 0 Param 4=&UniformRangedFloat(Minimum=3, Maximum=5)
    Module.Param0WriteEnabled=false
    Module.Param1WriteEnabled=true
    Module.Param2WriteEnabled=false
    Module.Param3WriteEnabled=true
   )
   &ShapeLocation(MaskSpawnGroup=false, Sphere Distribution=::NewEnumerator3, Sphere Radius=3, UseOwnerScale=true)
   &AddVelocity(
    Cone Axis=<-1 0 0>
    Module.UseConeFalloff=false
    Module.UseVelDistribution=false
    Override Seed=false
    Velocity Mode=::NewEnumerator2
    Velocity Speed=&RandomRangeFloat(Minimum=10, Maximum=25)
   )
  ]
  particleUpdateScript [
   &ParticleState()
   &Drag(Drag=0.8, Use Linear Drag=true, Use Rotational Drag=false)
   &ScaleColor(
    Scale Alpha=&FloatFromCurve(FloatCurve=DataInterface() {propertyValues {Curve {keys [<at 0 1 RCIM_Cubic>, {arriveTangent=-1, interpMode=RCIM_Cubic, leaveTangent=-1, time=0.6, value=0.1}, <at 1 0 RCIM_Cubic>]}}})
    Scale RGB=&VectorFromFloat(Value=&FloatFromCurve(FloatCurve=DataInterface() {propertyValues {Curve {keys [<at 0 1 RCIM_Cubic>,<at 0.1 10 RCIM_Cubic>,<at 1 1 RCIM_Cubic>]}}}))
    ScaleA=true
    ScaleRGB=true
   )
   &SolveForcesAndVelocity(Clamp Velocity=false, Limit Acceleration=false)
   &DynamicMaterialParameters(
    Index 0 Param 1=&FloatFromCurve(FloatCurve=DataInterface() {propertyValues {Curve {keys [<at 0.0017 0.9908 RCIM_Cubic>, {arriveTangent=0.0101, interpMode=RCIM_Cubic, leaveTangent=0.0101, time=0.3821, value=0.8044}, <at 0.9074 1>]}}})
    Index 0 Param 3=&FloatFromCurve(FloatCurve=DataInterface() {propertyValues {Curve {keys [<at 0 3 RCIM_Cubic>,<at 0.4 2 RCIM_Cubic>,<at 0.9928 3>]}}}, Scale Curve=8)
    Index 0 Param 4=1
    Module.Param0WriteEnabled=true
    Module.Param1WriteEnabled=true
    Module.Param2WriteEnabled=true
    Module.Param3WriteEnabled=true
   ) {enabled=false, module=DynamicMaterialParameters001}
   &ScaleSpriteSize(Scale Factor=/Niagara/DynamicInputs/ScaleByCurve/ScaleVector2DByCurve(
     InitialValue:Vector2f=<1 1>
     ScaleCurve:NiagaraDataInterfaceVectorCurve=DataInterface() {propertyValues {
       XCurve {keys [
         {interpMode=RCIM_Cubic, tangentMode=RCTM_User, time=-0.0064, value=0.3107}
         {arriveTangent=0.5504, interpMode=RCIM_Cubic, leaveTangent=0.5504, tangentMode=RCTM_User, time=0.181, value=0.4375}
         {interpMode=RCIM_Cubic, tangentMode=RCTM_User, time=0.9915, value=0.5417}
        ]}
       YCurve {keys [
         {interpMode=RCIM_Cubic, tangentMode=RCTM_User, time=-0.0064, value=0.3107}
         {arriveTangent=0.7396, interpMode=RCIM_Cubic, leaveTangent=0.7396, tangentMode=RCTM_User, time=0.1806, value=0.4402}
         {interpMode=RCIM_Cubic, tangentMode=RCTM_User, time=0.9874, value=0.5437}
        ]}
       ZCurve {keys []}
      }}
     CurveIndex:NiagaraFloat=$Particles.NormalizedAge
    ))
   &DynamicMaterialParameters(
    Index 0 Param 1=&FloatFromCurve(FloatCurve=DataInterface() {propertyValues {Curve {keys [{arriveTangent=2.7551, interpMode=RCIM_Cubic, leaveTangent=2.7551, tangentMode=RCTM_User, time=0, value=0}, <at 0.7 1>]}}})
    Index 0 Param 2=1
    Index 0 Param 3=1
    Index 0 Param 4=2
    Module.Param0WriteEnabled=true
    Module.Param1WriteEnabled=true
    Module.Param2WriteEnabled=true
    Module.Param3WriteEnabled=true
   ) {module=DynamicMaterialParameters002}
   /Niagara/Modules/Update/SubUV/SubUVAnimation(Number Of Frames=&FloatFromCurve(FloatCurve=DataInterface() {propertyValues {Curve {keys [<at 0.0156 7.4642 RCIM_Cubic>,<at 1.0068 39.7662>]}}}))
   /SetVariables(Particles.VisibilityTag=&MakeCustomIntFromBool(Bool=$User.bisHeadHidden, True INT=1, False INT=0)) {isSetParameters=true, module=SetVariables_#1}
  ]
  rendererClasses [&NiagaraSpriteRendererProperties]
  renderers [
   {rendererClass=&NiagaraSpriteRendererProperties}
   {rendererClass=&NiagaraSpriteRendererProperties, rendererIndex=1}
  ]
 }
 FlamingHeadMesh {
  emitterUpdateScript [
   &EmitterState(Life Cycle Mode=::NewEnumerator1, Loop Behavior=::NewEnumerator1, Loop Duration Mode=::NewEnumerator1, UseLoopDelay=false)
   &SpawnBurst_Instantaneous(Spawn Count=2, Use Spawn Probability=false, UseLoopCountLimit=false)
  ]
  particleSpawnScript [&InitializeParticle(
    Color=<rgba 0.2024 0.5 0.0649 1>
    Lifetime=5
    Mesh Scale=&MakeVector(
     X=&SelectFloatFromArray(
      Direct Array Index=&ReturnExecIndex()
      Float Selection Array=DataInterface() {propertyValues {FloatData [1,-1]}}
     )
     Y=1
     Z=1
    )
    Mesh Scale Mode=::NewEnumerator4
    Position Offset=<1 0 1.5>
    Position Offset Coordinate Space=::Local
    UsePositionOffset=true
    Write Material Random=true
   )]
  particleUpdateScript [
   &ParticleState(Kill Particles When Lifetime Has Elapsed=false, Loop Particles Lifetime=true)
   &CameraOffset() {enabled=false}
   &DynamicMaterialParameters(
    Index 0 Param 1=&Lerp_Float(
     A=15
     B=18
     Alpha=&Waveform(
      Waveform Count=::NewEnumerator1
      "Waveform [1]"=::NewEnumerator12
      "[1] Phase Scale"=10
      "[1] Frequency"=1
      "[1] Amplitude Min/Max"=<0 1>
      "[1] Scale"=1.5
      "Waveform [2] Blend"=::NewEnumerator2
      "Waveform [2]"=::NewEnumerator12
      "[2] Input Position"=0
      "[2] Phase"=$Emitter.Age
      "[2] Phase Scale"=5
      "[2] Compound Frequency"=2
      "[2] Frequency"=1.5
      "[2] Amplitude Min/Max"=<0 1>
      "[2] Scale"=1
      "[2] Bias"=0
      "[2] Exponent"=1
      "[2] Pulse Duration"=0.1
      "[2] Clip Triangle Min/Max"=<0 0>
      "[2] Random Seed"=0
     )
    )
    Index 0 Param 2=&SelectFloatFromArray(
     Direct Array Index=&ReturnExecIndex()
     Direct Array Mode=::NewEnumerator1
     Float Selection Array=DataInterface() {propertyValues {FloatData [0,0.7376]}}
    )
    Index 0 Param 3=&SelectFloatFromArray(
     Direct Array Index=&ReturnExecIndex()
     Float Selection Array=DataInterface() {propertyValues {FloatData [-1,1]}}
    )
    Module.Param0WriteEnabled=true
    Module.Param1WriteEnabled=true
    Module.Param2WriteEnabled=true
    Module.Param3WriteEnabled=true
   )
   /SetVariables(Particles.VisibilityTag=&MakeCustomIntFromBool(Bool=$User.bisHeadHidden, True INT=1, False INT=0)) {isSetParameters=true, module=SetVariables_#1}
  ]
  rendererClasses [&NiagaraMeshRendererProperties]
  renderers [{rendererClass=&NiagaraMeshRendererProperties, rendererIndex=0}]
 }
 FlamingHead_Vertical {
  emitterUpdateScript [
   &EmitterState(Life Cycle Mode=::NewEnumerator1, Loop Behavior=::NewEnumerator1, Loop Duration Mode=::NewEnumerator1, UseLoopDelay=false)
   &SpawnBurst_Instantaneous(Spawn Count=2, Use Spawn Probability=false, UseLoopCountLimit=false)
  ]
  particleSpawnScript [
   &InitializeParticle(
    Color=<rgba 0.5011 2 0 1>
    Lifetime=5
    Mesh Scale=<0.125 1 1>
    Mesh Scale Mode=::NewEnumerator4
    Position Offset=&SelectVectorFromArray(
     Direct Array Index=&ReturnExecIndex()
     Vector Selection Array=DataInterface() {propertyValues {FloatData [<4 0 11>, <-2 0 0>]}}
    )
    Position Offset Coordinate Space=::Local
    UsePositionOffset=true
    Write Material Random=true
   )
   &InitialMeshOrientation(
    Mesh Orientation Mode=::NewEnumerator2
    Orientation Vector=&SelectVectorFromArray(
     Direct Array Index=&ReturnExecIndex()
     Vector Selection Array=DataInterface() {propertyValues {FloatData [<1 0 0>, <1 0 -0.3003>]}}
    )
    Use Rotation Vector=false
   )
  ]
  particleUpdateScript [
   &ParticleState(Kill Particles When Lifetime Has Elapsed=false, Loop Particles Lifetime=true)
   &CameraOffset(Camera Offset Amount=1)
   &DynamicMaterialParameters(Module.Param0WriteEnabled=true, Module.Param1WriteEnabled=true, Module.Param2WriteEnabled=true, Module.Param3WriteEnabled=true)
   /SetVariables(Particles.VisibilityTag=&MakeCustomIntFromBool(Bool=$User.bisHeadHidden, True INT=1, False INT=0)) {isSetParameters=true, module=SetVariables_#1}
  ]
  rendererClasses [&NiagaraMeshRendererProperties]
  renderers [{rendererClass=&NiagaraMeshRendererProperties, rendererIndex=0}]
 }
 Light {
  emitterUpdateScript [
   &EmitterState(
    Inactive Response=::NewEnumerator1
    Life Cycle Mode=::NewEnumerator1
    Loop Behavior=::NewEnumerator1
    Loop Duration Mode=::NewEnumerator1
    UseLoopDelay=false
   )
   &SpawnBurst_Instantaneous(Use Spawn Probability=false, UseLoopCountLimit=false)
  ]
  particleSpawnScript [
   &InitializeParticle(
    Color=<rgba 0.5885 1 0.1771 1>
    Lifetime=$Emitter.CurrentLoopDuration
    Position Offset=<-20 0 7>
    Sprite Size Mode=::NewEnumerator3
    UsePositionOffset=true
    Write Material Random=true
   )
   /SetVariables() {isSetParameters=true}
  ]
  particleUpdateScript [
   &ParticleState(
    Kill Particles When Lifetime Has Elapsed=false
    Let Infinitely Lived Particles Die When Emitter Deactivates=true
    Loop Particles Lifetime=true
   )
   &SolveForcesAndVelocity(Clamp Velocity=false, Limit Acceleration=false)
   &FN_Gen5_Light(
    Attenutation=1
    Intensity:NiagaraFloat=&Waveform(
     Waveform Count=::NewEnumerator1
     Global Amplitude Scale=800
     "Waveform [1]"=::NewEnumerator12
     "[1] Phase Scale"=2
     "[1] Frequency"=4
     "[1] Amplitude Min/Max"=<0 1>
     "[2] Average"=false
     "Waveform [2]"=::NewEnumerator12
     "[2] Input Position"=0
     "[2] Phase"=$Emitter.Age
     "[2] Phase Scale"=2
     "[2] Compound Frequency"=2
     "[2] Frequency"=6
     "[2] Amplitude Min/Max"=<0 1>
     "[2] Scale"=1
     "[2] Bias"=0
     "[2] Exponent"=1
     "[2] Pulse Duration"=0.1
     "[2] Clip Triangle Min/Max"=<0 0>
     "[2] Random Seed"=0
    )
   )
   /SetVariables(Particles.LightRadius=15) {isSetParameters=true}
   /SetVariables(Particles.LightExponent=1) {isSetParameters=true}
   &ScaleColor(
    Scale RGB=&VectorFromFloat(Value=&Multiply_Float(
      A=$Particles.FN_Gen5_Light.Intensity
      B=&MakeCustomFloatFromBool(Bool=$User.FrontEnd, True Float=2, False Float=1)
     ))
    ScaleA=true
    ScaleRGB=true
   )
   &JitterPosition(Jitter Amount=0.5) {enabled=false}
   /SetVariables(Particles.smoothed_offset:Vector3f=/Niagara/DynamicInputs/Lerp/SmoothLerpOverTime_Vector(
     Convergence Rate:NiagaraFloat=1
     Smooth Value:Vector3f=$Particles.smoothed_offset
     Target Value:Vector3f=&RandomVector(Vector Scale=17.5676, Constrain to Hemisphere=true, Hemisphere Constraint Vector=<1 0 0>, Evaluation Type=::NewEnumerator1)
    )) {isSetParameters=true}
   /SetVariables(Particles.Position=&AddVectorToPosition(Position=$Particles.Initial.Position, Vector=$Particles.smoothed_offset)) {isSetParameters=true}
   /SetVariables(Particles.LightEnabled:NiagaraBool=&Boolean_NotOperation(Bool=$User.bisHeadHidden)) {isSetParameters=true}
  ]
  rendererClasses [&NiagaraLightRendererProperties, &NiagaraSpriteRendererProperties]
  renderers [
   {rendererClass=&NiagaraLightRendererProperties, rendererIndex=0}
   {rendererClass=&NiagaraSpriteRendererProperties, rendererIndex=1}
  ]
 }
}
summary {
 systemName=NS_F_MED_bariumDemon_Head_Idle
 emitters [
  {emitterName=FlamingHeadMesh, rendererClasses [&NiagaraMeshRendererProperties]}
  {emitterName=FlamingHead_Vertical, rendererClasses [&NiagaraMeshRendererProperties]}
  {emitterName=Flames, rendererClasses [&NiagaraSpriteRendererProperties]}
  {emitterName=Light, rendererClasses [&NiagaraLightRendererProperties, &NiagaraSpriteRendererProperties]}
 ]
 userVariables [
  User.bisHeadHidden {type=NiagaraBool}
  User.FrontEnd {type=NiagaraBool, defaultValue=true}
  User.SkeletalMesh {
   type=NiagaraDataInterfaceSkeletalMesh
   defaultValue {
    struct=NiagaraExt_VariableValue_DataInterface
    value {dataInterface=&self:NiagaraDataInterfaceSkeletalMesh_0}
   }
  }
 ]
}
