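# Post-process-style Control Rig for the F_MED_BariumDemon female body that drives corrective dfrm_ bones for ribs, clavicles, shoulder pads, hips and thighs.
# Spherical pose readers on clavicles, upper arms and thighs write alpha variables, some multiplied into combined poses.
# Those alphas, some smoothed, weight bone rotation, translation and scale offsets, while aim constraints orient shoulder pads and thighs.
# Rifle, crouch, skydive and front-end flags switch or add adjustments.
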ControlRigBlueprint /BRCosmetics/Characters/Player/Female/Medium/Bodies/F_MED_BariumDemon/Meshes/F_MED_BariumDemon_CtrlRig
graphs [
 ContainedGraph {nodes [
   n1 = RigVMFunctionEntryNode()
   SequenceAggregate = &RigVMFunction_Sequence(ExecuteContext=@n1.ExecuteContext)
   SequenceAggregate_1 = &RigVMFunction_Sequence(ExecuteContext=@SequenceAggregate.B)
   SequenceAggregate_2 = &RigVMFunction_Sequence(ExecuteContext=@SequenceAggregate_1.B)
   RigVMFunctionReturnNode(A=@SequenceAggregate.A, B=@SequenceAggregate_1.A, C=@SequenceAggregate_2.A, D=@SequenceAggregate_2.B)
  ]}
 ContainedGraph {nodes [
   n1 = RigVMFunctionEntryNode()
   SequenceAggregate_1 = &RigVMFunction_Sequence(ExecuteContext=@n1.ExecuteContext)
   SequenceAggregate_1_1 = &RigVMFunction_Sequence(ExecuteContext=@SequenceAggregate_1.B)
   SequenceAggregate_1_2 = &RigVMFunction_Sequence(ExecuteContext=@SequenceAggregate_1_1.B)
   RigVMFunctionReturnNode(A=@SequenceAggregate_1.A, B=@SequenceAggregate_1_1.A, C=@SequenceAggregate_1_2.A, D=@SequenceAggregate_1_2.B)
  ]}
 ContainedGraph {nodes [
   n1 = RigVMFunctionEntryNode()
   SequenceAggregate_3 = &RigVMFunction_Sequence(ExecuteContext=@n1.ExecuteContext)
   SequenceAggregate_3_1 = &RigVMFunction_Sequence(ExecuteContext=@SequenceAggregate_3.B)
   SequenceAggregate_3_2 = &RigVMFunction_Sequence(ExecuteContext=@SequenceAggregate_3_1.B)
   RigVMFunctionReturnNode(A=@SequenceAggregate_3.A, B=@SequenceAggregate_3_1.A, C=@SequenceAggregate_3_2.A, D=@SequenceAggregate_3_2.B)
  ]}
 ContainedGraph {nodes [
   n1 = RigVMFunctionEntryNode()
   SequenceAggregate_1_1_1 = &RigVMFunction_Sequence(ExecuteContext=@n1.ExecuteContext)
   SequenceAggregate_1_1_1_1 = &RigVMFunction_Sequence(ExecuteContext=@SequenceAggregate_1_1_1.B)
   SequenceAggregate_1_1_1_2 = &RigVMFunction_Sequence(ExecuteContext=@SequenceAggregate_1_1_1_1.B)
   RigVMFunctionReturnNode(A=@SequenceAggregate_1_1_1.A, B=@SequenceAggregate_1_1_1_1.A, C=@SequenceAggregate_1_1_1_2.A, D=@SequenceAggregate_1_1_1_2.B)
  ]}
 ContainedGraph {nodes [
   n1 = RigVMFunctionEntryNode()
   SequenceAggregate_4 = &RigVMFunction_Sequence(ExecuteContext=@n1.ExecuteContext)
   SequenceAggregate_4_1 = &RigVMFunction_Sequence(ExecuteContext=@SequenceAggregate_4.B)
   RigVMFunctionReturnNode(A=@SequenceAggregate_4.A, B=@SequenceAggregate_4_1.A, C=@SequenceAggregate_4_1.B)
  ]}
 RigVMFunctionLibrary {links [], nodes []}
 RigVMModel {nodes [
   n1 = &RigUnit_BeginExecution()
   SequenceAggregate = RigVMAggregateNode(ExecuteContext=@n1.ExecutePin)
   SequenceAggregate_1 = RigVMAggregateNode(ExecuteContext=@SequenceAggregate.B)
   Branch_1_RigVMUnitNode = &RigVMFunction_ControlFlowBranch(ExecuteContext=@SequenceAggregate_1.C, Condition=RigVMVariableNode() {variable=has_rifle}.Value)
   RigVMVariableNode(Value=(X=0,Y=0,Z=0), ExecuteContext=@Branch_1_RigVMUnitNode.False) {variable=look_down_rib_mstr_02_l_rot}
   > RigVMVariableNode(Value=(X=0,Y=0,Z=0)) {variable=look_down_rib_mstr_03_l_rot}
   RigVMVariableNode(Value=(X=-2.5,Y=0,Z=2.5), ExecuteContext=@Branch_1_RigVMUnitNode.True) {variable=look_down_rib_mstr_02_l_rot}
   > RigVMVariableNode(Value=(X=-3.5,Y=0,Z=2.5)) {variable=look_down_rib_mstr_03_l_rot}
   n2 = &RigVMFunction_AlphaInterp(
    InterpSpeedDecreasing=10
    InterpSpeedIncreasing=10
    Scale=1
    bClampResult=False
    Value=RigVMVariableNode() {variable=spr_upperarm_l_down_alpha}.Value
   )
   &RigUnit_ModifyTransforms(
    ItemToModify=((Item=(Type=Bone,Name="dfrm_rib_mstr_01_l"),Transform=(Rotation=(X=0,Y=0,Z=0,W=1),Translation=(X=0,Y=0,Z=0),Scale3D=(X=1,Y=1,Z=1))))
    ItemToModify.0.Transform.Scale3D=RigVMVariableNode() {variable=upper_arm_dwn_rib_mstr_scl_fctr}.Value
    ExecutePin=@SequenceAggregate_1.D
    Weight=@n2.Result
   )
   > &RigUnit_ModifyTransforms(
    ItemToModify=((Item=(Type=Bone,Name="dfrm_rib_mstr_02_l"),Transform=(Rotation=(X=0,Y=0,Z=0,W=1),Translation=(X=0,Y=0,Z=0),Scale3D=(X=1,Y=1,Z=1))))
    ItemToModify.0.Transform.Rotation=&RigVMFunction_MathQuaternionFromEuler(Euler=RigVMVariableNode() {variable=look_down_rib_mstr_02_l_rot}.Value).Result
    ItemToModify.0.Transform.Scale3D=RigVMVariableNode() {variable=upper_arm_dwn_rib_mstr_scl_fctr}.Value
    Weight=@n2.Result
   )
   > &RigUnit_ModifyTransforms(
    ItemToModify=((Item=(Type=Bone,Name="dfrm_rib_mstr_03_l"),Transform=(Rotation=(X=0,Y=0,Z=0,W=1),Translation=(X=0,Y=0,Z=0),Scale3D=(X=1,Y=1,Z=1))))
    ItemToModify.0.Transform.Rotation=&RigVMFunction_MathQuaternionFromEuler(Euler=RigVMVariableNode() {variable=look_down_rib_mstr_03_l_rot}.Value).Result
    ItemToModify.0.Transform.Scale3D=RigVMVariableNode() {variable=upper_arm_dwn_rib_mstr_scl_fctr}.Value
    Weight=@n2.Result
   )
   > &RigUnit_ModifyTransforms(
    ItemToModify=((Item=(Type=Bone,Name="dfrm_rib_mstr_04_l"),Transform=(Rotation=(X=0,Y=0,Z=0,W=1),Translation=(X=0,Y=0,Z=0),Scale3D=(X=1,Y=1,Z=1))))
    ItemToModify.0.Transform.Scale3D=RigVMVariableNode() {variable=upper_arm_dwn_rib_mstr_scl_fctr}.Value
    Weight=@n2.Result
   )
   &RigUnit_RotationConstraintLocalSpaceOffset(
    AdvancedSettings=(RotationOrderForFilter=XZY)
    Child=(Type=Bone,Name="dfrm_spine_04")
    Filter=(bX=True,bY=True,bZ=True)
    Parents=((Item=(Type=Bone,Name="spine_03"),Weight=0.5),(Item=(Type=Bone,Name="spine_05"),Weight=0.5))
    Weight=1
    ExecutePin=@SequenceAggregate_1.A
   )
   &RigUnit_ModifyTransforms(
    ItemToModify=((Item=(Type=Bone,Name="dfrm_rib_mstr_01_l"),Transform=(Rotation=(X=0,Y=0,Z=0,W=1),Translation=(X=0,Y=0,Z=0),Scale3D=(X=1,Y=1,Z=1))))
    ItemToModify.0.Transform.Rotation=&RigVMFunction_MathQuaternionFromEuler(Euler=(X=0,Y=0,Z=-25)).Result
    Weight=RigVMVariableNode() {variable=spr_calvicle_l_up_alpha}.Value
    ExecutePin=@SequenceAggregate_1.B
   )
   > &RigUnit_ModifyTransforms(
    ItemToModify=((Item=(Type=Bone,Name="dfrm_rib_mstr_02_l"),Transform=(Rotation=(X=0,Y=0,Z=0,W=1),Translation=(X=0,Y=0,Z=0),Scale3D=(X=1,Y=1,Z=1))))
    ItemToModify.0.Transform.Rotation=&RigVMFunction_MathQuaternionFromEuler(Euler=(X=0,Y=0,Z=-12.5)).Result
    Weight=RigVMVariableNode() {variable=spr_calvicle_l_up_alpha}.Value
   )
   > &RigUnit_ModifyTransforms(
    ItemToModify=((Item=(Type=Bone,Name="dfrm_rib_mstr_03_l"),Transform=(Rotation=(X=0,Y=0,Z=0,W=1),Translation=(X=0,Y=0,Z=0),Scale3D=(X=1,Y=1,Z=1))))
    ItemToModify.0.Transform.Rotation=&RigVMFunction_MathQuaternionFromEuler(Euler=(X=0,Y=0,Z=-6.25)).Result
    Weight=RigVMVariableNode() {variable=spr_calvicle_l_up_alpha}.Value
   )
   > &RigUnit_ModifyTransforms(
    ItemToModify=((Item=(Type=Bone,Name="dfrm_rib_mstr_04_l"),Transform=(Rotation=(X=0,Y=0,Z=0,W=1),Translation=(X=0,Y=0,Z=0),Scale3D=(X=1,Y=1,Z=1))))
    ItemToModify.0.Transform.Rotation=&RigVMFunction_MathQuaternionFromEuler(Euler=(X=0,Y=0,Z=-3.15)).Result
    Weight=RigVMVariableNode() {variable=spr_calvicle_l_up_alpha}.Value
   )
   SequenceAggregate_1_1_1 = RigVMAggregateNode(ExecuteContext=@SequenceAggregate.C)
   n3 = &RigUnit_ModifyTransforms(
    ItemToModify=((Item=(Type=Bone,Name="dfrm_rib_mstr_01_r"),Transform=(Rotation=(X=0,Y=0,Z=0,W=1),Translation=(X=0,Y=0,Z=0),Scale3D=(X=1,Y=1,Z=1))))
    ItemToModify.0.Transform.Rotation=&RigVMFunction_MathQuaternionFromEuler(Euler=(X=0,Y=0,Z=-25)).Result
    ExecutePin=@SequenceAggregate_1_1_1.A
    Weight=RigVMVariableNode() {variable=spr_calvicle_r_up_alpha}.Value
   )
   SequenceAggregate_2 = &RigVMFunction_Sequence(ExecuteContext=@n3.ExecutePin)
   &RigUnit_ModifyTransforms(
    ItemToModify=((Item=(Type=Bone,Name="dfrm_rib_mstr_02_r"),Transform=(Rotation=(X=0,Y=0,Z=0,W=1),Translation=(X=0,Y=0,Z=0),Scale3D=(X=1,Y=1,Z=1))))
    ItemToModify.0.Transform.Rotation=&RigVMFunction_MathQuaternionFromEuler(Euler=(X=0,Y=0,Z=-12.5)).Result
    ExecutePin=@SequenceAggregate_2.A
    Weight=RigVMVariableNode() {variable=spr_calvicle_r_up_alpha}.Value
   )
   > &RigUnit_ModifyTransforms(
    ItemToModify=((Item=(Type=Bone,Name="dfrm_rib_mstr_03_r"),Transform=(Rotation=(X=0,Y=0,Z=0,W=1),Translation=(X=0,Y=0,Z=0),Scale3D=(X=1,Y=1,Z=1))))
    ItemToModify.0.Transform.Rotation=&RigVMFunction_MathQuaternionFromEuler(Euler=(X=0,Y=0,Z=-6.5)).Result
    Weight=RigVMVariableNode() {variable=spr_calvicle_r_up_alpha}.Value
   )
   > &RigUnit_ModifyTransforms(
    ItemToModify=((Item=(Type=Bone,Name="dfrm_rib_mstr_04_r"),Transform=(Rotation=(X=0,Y=0,Z=0,W=1),Translation=(X=0,Y=0,Z=0),Scale3D=(X=1,Y=1,Z=1))))
    ItemToModify.0.Transform.Rotation=&RigVMFunction_MathQuaternionFromEuler(Euler=(X=0,Y=0,Z=-3.15)).Result
    Weight=RigVMVariableNode() {variable=spr_calvicle_r_up_alpha}.Value
   )
   n4 = RigVMVariableNode() {variable=has_rifle}
   n5 = &RigVMFunction_AlphaInterp(
    InterpSpeedDecreasing=10
    InterpSpeedIncreasing=10
    Scale=1
    bClampResult=False
    Value=RigVMVariableNode() {variable=spr_upperarm_l_clavicle_r_aim_dwn_alpha}.Value
   )
   &RigUnit_ModifyTransforms(
    ItemToModify=((Item=(Type=Bone,Name="dfrm_rib_mstr_02_r"),Transform=(Rotation=(X=0,Y=0,Z=0,W=1),Translation=(X=0,Y=0,Z=0),Scale3D=(X=1,Y=1,Z=1))))
    ItemToModify.0.Transform.Rotation=&RigVMFunction_MathQuaternionFromEuler(Euler=(X=0,Y=0,Z=-6.5)).Result
    Weight=&RigVMFunction_MathDoubleMul(A=&RigVMFunction_MathBoolToFloat(Value=@n4.Value).Result, B=@n5.Result).Result
    ExecutePin=@SequenceAggregate_2.B
   )
   n6 = &RigVMFunction_AlphaInterp(
    InterpSpeedDecreasing=17.5
    InterpSpeedIncreasing=15
    Scale=1
    bClampResult=False
    bMapRange=true
    Value=RigVMVariableNode() {variable=spr_upperarm_l_clavicle_r_aim_dwn_alpha}.Value
   )
   &RigUnit_ModifyTransforms(
    ItemToModify=((Item=(Type=Bone,Name="dfrm_rib_mstr_01_r"),Transform=(Rotation=(X=0,Y=0,Z=0,W=1),Translation=(X=0,Y=0,Z=0),Scale3D=(X=1,Y=1,Z=1))))
    ItemToModify.0.Transform.Rotation=&RigVMFunction_MathQuaternionFromEuler(Euler=(X=0,Y=0,Z=0)).Result
    ExecutePin=@SequenceAggregate_1_1_1.B
    Weight=@n6.Result
   )
   > &RigUnit_ModifyTransforms(
    ItemToModify=((Item=(Type=Bone,Name="dfrm_rib_mstr_02_r"),Transform=(Rotation=(X=0,Y=0,Z=0,W=1),Translation=(X=0,Y=0,Z=0),Scale3D=(X=1,Y=1,Z=1))))
    ItemToModify.0.Transform.Rotation=&RigVMFunction_MathQuaternionFromEuler(Euler=(X=0,Y=0,Z=-3)).Result
    Weight=@n6.Result
   )
   > &RigUnit_ModifyTransforms(
    ItemToModify=((Item=(Type=Bone,Name="dfrm_rib_mstr_03_r"),Transform=(Rotation=(X=0,Y=0,Z=0,W=1),Translation=(X=0,Y=0,Z=0),Scale3D=(X=1,Y=1,Z=1))))
    ItemToModify.0.Transform.Rotation=&RigVMFunction_MathQuaternionFromEuler(Euler=(X=0,Y=0,Z=-5)).Result
    Weight=@n6.Result
   )
   > &RigUnit_ModifyTransforms(
    ItemToModify=((Item=(Type=Bone,Name="dfrm_rib_mstr_04_r"),Transform=(Rotation=(X=0,Y=0,Z=0,W=1),Translation=(X=0,Y=0,Z=0),Scale3D=(X=1,Y=1,Z=1))))
    ItemToModify.0.Transform.Rotation=&RigVMFunction_MathQuaternionFromEuler(Euler=(X=0,Y=0,Z=-5)).Result
    Weight=@n6.Result
   )
   &RigUnit_ModifyTransforms(
    ItemToModify=((Item=(Type=Bone,Name="dfrm_rib_mstr_01_r"),Transform=(Rotation=(X=0,Y=0,Z=0,W=1),Translation=(X=0,Y=0,Z=0),Scale3D=(X=1,Y=1,Z=1))))
    Weight=RigVMVariableNode() {variable=spr_upperarm_r_down_alpha}.Value
    ItemToModify.0.Transform.Scale3D=RigVMVariableNode() {variable=upper_arm_dwn_rib_mstr_scl_fctr}.Value
    ExecutePin=@SequenceAggregate_1_1_1.C
   )
   > &RigUnit_ModifyTransforms(
    ItemToModify=((Item=(Type=Bone,Name="dfrm_rib_mstr_02_r"),Transform=(Rotation=(X=0,Y=0,Z=0,W=1),Translation=(X=0,Y=0,Z=0),Scale3D=(X=1,Y=1,Z=1))))
    Weight=RigVMVariableNode() {variable=spr_upperarm_r_down_alpha}.Value
    ItemToModify.0.Transform.Scale3D=RigVMVariableNode() {variable=upper_arm_dwn_rib_mstr_scl_fctr}.Value
   )
   > &RigUnit_ModifyTransforms(
    ItemToModify=((Item=(Type=Bone,Name="dfrm_rib_mstr_03_r"),Transform=(Rotation=(X=0,Y=0,Z=0,W=1),Translation=(X=0,Y=0,Z=0),Scale3D=(X=1,Y=1,Z=1))))
    Weight=RigVMVariableNode() {variable=spr_upperarm_r_down_alpha}.Value
    ItemToModify.0.Transform.Scale3D=RigVMVariableNode() {variable=upper_arm_dwn_rib_mstr_scl_fctr}.Value
   )
   > &RigUnit_ModifyTransforms(
    ItemToModify=((Item=(Type=Bone,Name="dfrm_rib_mstr_04_r"),Transform=(Rotation=(X=0,Y=0,Z=0,W=1),Translation=(X=0,Y=0,Z=0),Scale3D=(X=1,Y=1,Z=1))))
    Weight=RigVMVariableNode() {variable=spr_upperarm_r_down_alpha}.Value
    ItemToModify.0.Transform.Scale3D=RigVMVariableNode() {variable=upper_arm_dwn_rib_mstr_scl_fctr}.Value
   )
   n7 = &RigVMFunction_AlphaInterp(
    InterpSpeedDecreasing=10
    InterpSpeedIncreasing=10
    Scale=1
    bClampResult=False
    Value=RigVMVariableNode() {variable=spr_clavicle_r_twist_alpha}.Value
   )
   &RigUnit_ModifyTransforms(
    ItemToModify=((Item=(Type=Bone,Name="dfrm_clavicle_r"),Transform=(Rotation=(X=0,Y=0,Z=0,W=1),Translation=(X=0,Y=0,Z=0),Scale3D=(X=1,Y=1,Z=1))))
    ItemToModify.0.Transform.Rotation=&RigVMFunction_MathQuaternionFromEuler(Euler=(X=-15,Y=-5,Z=-10)).Result
    Weight=@n7.Result
    ExecutePin=@SequenceAggregate_1_1_1.D
   )
   SequenceAggregate_3 = RigVMAggregateNode(ExecuteContext=@SequenceAggregate.D)
   SequenceAggregate_4 = RigVMAggregateNode(ExecuteContext=@SequenceAggregate_3.A)
   Branch_2_RigVMUnitNode = &RigVMFunction_ControlFlowBranch(ExecuteContext=@SequenceAggregate_4.A, Condition=RigVMVariableNode() {variable=is_skydiving}.Value)
   &RigUnit_ModifyTransforms(
    ItemToModify=((Item=(Type=Bone,Name="dfrm_shoulder_pad_up_aim_l"),Transform=(Rotation=(X=0,Y=0,Z=0,W=1),Translation=(X=-10,Y=0,Z=5),Scale3D=(X=1,Y=1,Z=1))))
    ExecutePin=@Branch_2_RigVMUnitNode.True
    Weight=RigVMVariableNode() {variable=spr_upperarm_l_skydive_alpha}.Value
   )
   > &RigUnit_ModifyTransforms(
    ItemToModify=((Item=(Type=Bone,Name="dfrm_shoulder_pad_up_aim_r"),Transform=(Rotation=(X=0,Y=0,Z=0,W=1),Translation=(X=-10,Y=0,Z=5),Scale3D=(X=1,Y=1,Z=1))))
    Weight=RigVMVariableNode() {variable=spr_upperarm_r_skydive_alpha}.Value
   )
   &RigUnit_ModifyTransforms(
    ItemToModify=((Item=(Type=Bone,Name="dfrm_shoulder_pad_up_aim_l"),Transform=(Rotation=(X=0,Y=0,Z=0,W=1),Translation=(X=5,Y=2.5,Z=0),Scale3D=(X=1,Y=1,Z=1))))
    ExecutePin=@SequenceAggregate_4.C
    Weight=RigVMVariableNode() {variable=spr_ADS_pistol_up_alpha}.Value
   )
   > &RigUnit_ModifyTransforms(
    ItemToModify=((Item=(Type=Bone,Name="dfrm_shoulder_pad_up_aim_r"),Transform=(Rotation=(X=0,Y=0,Z=0,W=1),Translation=(X=5,Y=2.5,Z=5),Scale3D=(X=1,Y=1,Z=1))))
    Weight=RigVMVariableNode() {variable=spr_ADS_pistol_up_alpha}.Value
   )
   > /Script/ControlRig.RigUnit_AimItem(
    DebugSettings=(Scale=10,WorldOffset=(Rotation=(X=0,Y=0,Z=0,W=1),Translation=(X=0,Y=0,Z=0),Scale3D=(X=1,Y=1,Z=1)))
    Item=(Type=Bone,Name="dfrm_shoulder_pad_mstr_l")
    Primary=(Weight=1,Axis=(X=1,Y=0,Z=0),Target=(X=0,Y=0,Z=0),Kind=Location,Space=(Type=Bone,Name="lowerarm_l"))
    Secondary=(Weight=1,Axis=(X=0,Y=-1,Z=0),Target=(X=0,Y=0,Z=0),Kind=Location,Space=(Type=Bone,Name="dfrm_shoulder_pad_up_aim_l"))
    Weight=1
   ) {is=struct RigVMUnitNode}
   > &RigUnit_AimItem(
    DebugSettings=(Scale=10,WorldOffset=(Rotation=(X=0,Y=0,Z=0,W=1),Translation=(X=0,Y=0,Z=0),Scale3D=(X=1,Y=1,Z=1)))
    Item=(Type=Bone,Name="dfrm_shoulder_pad_mstr_r")
    Primary=(Weight=1,Axis=(X=1,Y=0,Z=0),Target=(X=0,Y=0,Z=0),Kind=Location,Space=(Type=Bone,Name="lowerarm_r"))
    Secondary=(Weight=1,Axis=(X=0,Y=-1,Z=0),Target=(X=0,Y=0,Z=0),Kind=Location,Space=(Type=Bone,Name="dfrm_shoulder_pad_up_aim_r"))
    Weight=1
   )
   Branch_RigVMUnitNode = &RigVMFunction_ControlFlowBranch(ExecuteContext=@SequenceAggregate_3.D, Condition=RigVMVariableNode() {variable=is_frontEnd}.Value)
   &RigUnit_ModifyTransforms(
    ItemToModify=((Item=(Type=Bone,Name="dfrm_clavicle_l"),Transform=(Rotation=(X=0,Y=0,Z=0,W=1),Translation=(X=0,Y=0,Z=0),Scale3D=(X=0.7,Y=1,Z=1))))
    Weight=1
    ExecutePin=@Branch_RigVMUnitNode.True
   )
   > &RigUnit_ModifyTransforms(
    ItemToModify=((Item=(Type=Bone,Name="dfrm_clavicle_r"),Transform=(Rotation=(X=0,Y=0,Z=0,W=1),Translation=(X=0,Y=0,Z=0),Scale3D=(X=0.7,Y=1,Z=1))))
    Weight=1
   )
   &RigUnit_ModifyTransforms(
    ItemToModify=((Item=(Type=Bone,Name="dfrm_hip_adjust_l"),Transform=(Rotation=(X=0,Y=0,Z=0,W=1),Translation=(X=0,Y=0,Z=0),Scale3D=(X=1,Y=0.9,Z=1))))
    ItemToModify.0.Transform.Rotation=&RigVMFunction_MathQuaternionFromEuler(Euler=(X=-25,Y=0,Z=-10)).Result
    Weight=RigVMVariableNode() {variable=spr_thigh_l_crouch_alpha}.Value
    ExecutePin=@SequenceAggregate_3.B
   )
   > &RigUnit_AimItem(
    DebugSettings=(Scale=10,WorldOffset=(Rotation=(X=0,Y=0,Z=0,W=1),Translation=(X=0,Y=0,Z=0),Scale3D=(X=1,Y=1,Z=1)))
    Item=(Type=Bone,Name="dfrm_thigh_l")
    Primary=(Weight=1,Axis=(X=1,Y=0,Z=0),Target=(X=0,Y=0,Z=0),Kind=Location,Space=(Type=Bone,Name="calf_l"))
    Secondary=(Weight=1,Axis=(X=0,Y=0,Z=-1),Target=(X=0,Y=0,Z=0),Kind=Location,Space=(Type=Bone,Name="dfrm_thigh_up_aim_l"))
    Weight=1
   )
   > &RigUnit_ModifyTransforms(
    ItemToModify=((Item=(Type=Bone,Name="dfrm_hip_adjust_r"),Transform=(Rotation=(X=0,Y=0,Z=0,W=1),Translation=(X=0,Y=0,Z=0),Scale3D=(X=1,Y=0.9,Z=1))))
    ItemToModify.0.Transform.Rotation=&RigVMFunction_MathQuaternionFromEuler(Euler=(X=25,Y=0,Z=-10)).Result
    Weight=RigVMVariableNode() {variable=spr_thigh_r_crouch_alpha}.Value
   )
   > &RigUnit_AimItem(
    DebugSettings=(Scale=10,WorldOffset=(Rotation=(X=0,Y=0,Z=0,W=1),Translation=(X=0,Y=0,Z=0),Scale3D=(X=1,Y=1,Z=1)))
    Item=(Type=Bone,Name="dfrm_thigh_r")
    Primary=(Weight=1,Axis=(X=1,Y=0,Z=0),Target=(X=0,Y=0,Z=0),Kind=Location,Space=(Type=Bone,Name="calf_r"))
    Secondary=(Weight=1,Axis=(X=0,Y=0,Z=1),Target=(X=0,Y=0,Z=0),Kind=Location,Space=(Type=Bone,Name="dfrm_thigh_up_aim_r"))
    Weight=1
   )
   AlphaInterp = &RigVMFunction_AlphaInterp(
    InterpSpeedDecreasing=10
    InterpSpeedIncreasing=10
    Scale=1
    bClampResult=False
    Value=&RigVMFunction_MathBoolToFloat(Value=RigVMVariableNode() {variable=is_crouching}.Value).Result
   )
   &RigUnit_ModifyTransforms(
    ItemToModify=((Item=(Type=Bone,Name="dfrm_hip_adjust_r"),Transform=(Rotation=(X=0,Y=0,Z=0,W=1),Translation=(X=0,Y=0,Z=0),Scale3D=(X=1,Y=1,Z=1))))
    Weight=@AlphaInterp.Result
    ItemToModify.0.Transform.Rotation=&RigVMFunction_MathQuaternionFromEuler(Euler=(X=12.5,Y=0,Z=-7.5)).Result
    ExecutePin=@SequenceAggregate_3.C
   )
   n8 = &RigUnit_SphericalPoseReader(
    ActiveRegionScaleFactors=(PositiveWidth=0.3,NegativeWidth=2.5,PositiveHeight=2.5,NegativeHeight=2.5)
    ActiveRegionSize=0.1
    Debug=(bDrawDebug=false,DebugScale=25,DebugSegments=20,DebugThickness=0.25)
    DriverAxis=(X=1,Y=0,Z=0)
    DriverItem=(Type=Bone,Name="clavicle_l")
    FalloffRegionScaleFactors=(PositiveWidth=0.95,NegativeWidth=1,PositiveHeight=1,NegativeHeight=1)
    FalloffSize=0.25
    OptionalParentItem=(Type=Bone)
    RotationOffset=(X=0,Y=-45.5,Z=8.5)
    ExecutePin=@SequenceAggregate.A
   )
   > RigVMVariableNode(Value=@n8.OutputParam) {variable=spr_calvicle_l_up_alpha}
   > n9 = &RigUnit_SphericalPoseReader(
    ActiveRegionScaleFactors=(PositiveWidth=3.1,NegativeWidth=3.1,PositiveHeight=3.1,NegativeHeight=1)
    ActiveRegionSize=0.08
    Debug=(bDrawDebug=false,bDrawLocalAxes=false,DebugScale=25,DebugSegments=20,DebugThickness=0.25)
    DriverAxis=(X=1,Y=0,Z=0)
    DriverItem=(Type=Bone,Name="upperarm_l")
    FalloffRegionScaleFactors=(PositiveWidth=1,NegativeWidth=1,PositiveHeight=1,NegativeHeight=0.45)
    FalloffSize=0.25
    OptionalParentItem=(Type=Bone)
    RotationOffset=(X=117.5,Y=0,Z=-75)
   )
   > RigVMVariableNode(Value=@n9.OutputParam) {variable=spr_upperarm_l_down_alpha}
   > n10 = &RigUnit_SphericalPoseReader(
    ActiveRegionScaleFactors=(PositiveWidth=0.3,NegativeWidth=2.5,PositiveHeight=2.5,NegativeHeight=2.5)
    ActiveRegionSize=0.1
    Debug=(bDrawDebug=false,DebugScale=25,DebugSegments=20,DebugThickness=0.25)
    DriverAxis=(X=-1,Y=0,Z=0)
    DriverItem=(Type=Bone,Name="clavicle_r")
    FalloffRegionScaleFactors=(PositiveWidth=0.95,NegativeWidth=1,PositiveHeight=1,NegativeHeight=1)
    FalloffSize=0.25
    FlipHeightScaling=false
    FlipWidthScaling=false
    OptionalParentItem=(Type=Bone)
    RotationOffset=(X=0,Y=136.5,Z=8.5)
   )
   > RigVMVariableNode(Value=@n10.OutputParam) {variable=spr_calvicle_r_up_alpha}
   > n11 = &RigUnit_SphericalPoseReader(
    ActiveRegionScaleFactors=(PositiveWidth=5,NegativeWidth=1,PositiveHeight=3,NegativeHeight=5)
    ActiveRegionSize=0.05
    Debug=(bDrawDebug=false,DebugScale=25,DebugSegments=20,DebugThickness=0.25)
    DriverAxis=(X=-1,Y=0,Z=0)
    DriverItem=(Type=Bone,Name="upperarm_r")
    FalloffRegionScaleFactors=(PositiveWidth=1,NegativeWidth=1,PositiveHeight=2.5,NegativeHeight=1)
    FalloffSize=0.3
    FlipHeightScaling=true
    FlipWidthScaling=false
    OptionalParentItem=(Type=Bone)
    RotationOffset=(X=-25,Y=25,Z=0)
   )
   > RigVMVariableNode(Value=@n11.OutputParam) {variable=spr_upperarm_r_down_alpha}
   > n12 = &RigUnit_SphericalPoseReader(
    ActiveRegionScaleFactors=(PositiveWidth=1,NegativeWidth=1,PositiveHeight=1,NegativeHeight=1)
    ActiveRegionSize=0.075
    Debug=(bDrawDebug=false,DebugScale=25,DebugSegments=20,DebugThickness=0.25)
    DriverAxis=(X=1,Y=0,Z=0)
    DriverItem=(Type=Bone,Name="upperarm_l")
    FalloffRegionScaleFactors=(PositiveWidth=1,NegativeWidth=1,PositiveHeight=1,NegativeHeight=1)
    FalloffSize=0.2
    OptionalParentItem=(Type=Bone)
    RotationOffset=(X=42,Y=-50,Z=0)
   )
   > n13 = &RigUnit_SphericalPoseReader(
    ActiveRegionScaleFactors=(PositiveWidth=1,NegativeWidth=1,PositiveHeight=1,NegativeHeight=1)
    ActiveRegionSize=0.075
    Debug=(bDrawDebug=false,DebugScale=25,DebugSegments=20,DebugThickness=0.25)
    DriverAxis=(X=-1,Y=0,Z=0)
    DriverItem=(Type=Bone,Name="upperarm_r")
    FalloffRegionScaleFactors=(PositiveWidth=1,NegativeWidth=1,PositiveHeight=1,NegativeHeight=1)
    FalloffSize=0.2
    OptionalParentItem=(Type=Bone)
    RotationOffset=(X=-82.5,Y=0,Z=-15)
   )
   > RigVMVariableNode(Value=&RigVMFunction_MathDoubleMul(A=@n13.OutputParam, B=@n12.OutputParam).Result) {variable=spr_ADS_pistol_up_alpha}
   > n14 = &RigUnit_SphericalPoseReader(
    ActiveRegionScaleFactors=(PositiveWidth=1,NegativeWidth=1,PositiveHeight=1,NegativeHeight=1)
    ActiveRegionSize=0.05
    Debug=(bDrawDebug=false,DebugScale=25,DebugSegments=20,DebugThickness=0.25)
    DriverAxis=(X=-1,Y=0,Z=0)
    DriverItem=(Type=Bone,Name="clavicle_r")
    FalloffRegionScaleFactors=(PositiveWidth=0.5,NegativeWidth=1,PositiveHeight=1,NegativeHeight=1)
    FalloffSize=0.15
    FlipHeightScaling=false
    FlipWidthScaling=false
    OptionalParentItem=(Type=Bone)
    RotationOffset=(X=0,Y=114,Z=14)
   )
   > n15 = &RigUnit_SphericalPoseReader(
    ActiveRegionScaleFactors=(PositiveWidth=1.5,NegativeWidth=0.9,PositiveHeight=1.25,NegativeHeight=1.25)
    ActiveRegionSize=0.085
    Debug=(bDrawDebug=false,DebugScale=25,DebugSegments=20,DebugThickness=0.25)
    DriverAxis=(X=1,Y=0,Z=0)
    DriverItem=(Type=Bone,Name="upperarm_l")
    FalloffRegionScaleFactors=(PositiveWidth=1.5,NegativeWidth=0.8,PositiveHeight=1,NegativeHeight=1)
    FalloffSize=0.175
    OptionalParentItem=(Type=Bone)
    RotationOffset=(X=0,Y=-122.5,Z=15)
   )
   Multiply = &RigVMFunction_MathDoubleMul(A=@n15.OutputParam, B=@n14.OutputParam)
   > RigVMVariableNode(Value=@Multiply.Result) {variable=spr_upperarm_l_clavicle_r_aim_dwn_alpha}
   > n16 = &RigUnit_SphericalPoseReader(
    ActiveRegionScaleFactors=(PositiveWidth=0.6,NegativeWidth=2.65,PositiveHeight=2.65,NegativeHeight=2.65)
    ActiveRegionSize=0.15
    Debug=(bDrawDebug=false,DebugScale=25,DebugSegments=20,DebugThickness=0.25)
    DriverAxis=(X=-1,Y=0,Z=0)
    DriverItem=(Type=Bone,Name="upperarm_r")
    FalloffRegionScaleFactors=(PositiveWidth=0.8,NegativeWidth=1,PositiveHeight=1,NegativeHeight=1)
    FalloffSize=0.4
    OptionalParentItem=(Type=Bone)
    RotationOffset=(X=-20,Y=160,Z=20)
   )
   > RigVMVariableNode(Value=@n16.OutputParam) {variable=spr_upperarm_r_skydive_alpha}
   > n17 = &RigUnit_SphericalPoseReader(
    ActiveRegionScaleFactors=(PositiveWidth=0.6,NegativeWidth=2.65,PositiveHeight=2.65,NegativeHeight=2.65)
    ActiveRegionSize=0.15
    Debug=(bDrawDebug=false,DebugScale=25,DebugSegments=20,DebugThickness=0.25)
    DriverAxis=(X=1,Y=0,Z=0)
    DriverItem=(Type=Bone,Name="upperarm_l")
    FalloffRegionScaleFactors=(PositiveWidth=0.8,NegativeWidth=1,PositiveHeight=1,NegativeHeight=1)
    FalloffSize=0.4
    OptionalParentItem=(Type=Bone)
    RotationOffset=(X=20,Y=-20,Z=20)
   )
   > RigVMVariableNode(Value=@n17.OutputParam) {variable=spr_upperarm_l_skydive_alpha}
   > n18 = &RigUnit_SphericalPoseReader(
    ActiveRegionScaleFactors=(PositiveWidth=1,NegativeWidth=1,PositiveHeight=1,NegativeHeight=1)
    ActiveRegionSize=0.125
    Debug=(bDrawDebug=false,DebugScale=25,DebugSegments=20,DebugThickness=0.25)
    DriverAxis=(X=-1,Y=0,Z=0)
    DriverItem=(Type=Bone,Name="thigh_l")
    FalloffRegionScaleFactors=(PositiveWidth=1,NegativeWidth=1,PositiveHeight=1,NegativeHeight=1)
    FalloffSize=0.35
    OptionalParentItem=(Type=Bone)
    RotationOffset=(X=115,Y=25,Z=0)
   )
   > RigVMVariableNode(Value=@n18.OutputParam) {variable=spr_thigh_l_crouch_alpha}
   > n19 = &RigUnit_SphericalPoseReader(
    ActiveRegionScaleFactors=(PositiveWidth=1,NegativeWidth=1,PositiveHeight=1,NegativeHeight=1)
    ActiveRegionSize=0.125
    Debug=(bDrawDebug=false,DebugScale=25,DebugSegments=20,DebugThickness=0.25)
    DriverAxis=(X=1,Y=0,Z=0)
    DriverItem=(Type=Bone,Name="thigh_r")
    FalloffRegionScaleFactors=(PositiveWidth=1,NegativeWidth=1,PositiveHeight=1,NegativeHeight=1)
    FalloffSize=0.35
    OptionalParentItem=(Type=Bone)
    RotationOffset=(X=-65,Y=25,Z=0)
   )
   > RigVMVariableNode(Value=@n19.OutputParam) {variable=spr_thigh_r_crouch_alpha}
   > n20 = &RigUnit_SphericalPoseReader(
    ActiveRegionScaleFactors=(PositiveWidth=1,NegativeWidth=1,PositiveHeight=1,NegativeHeight=1)
    ActiveRegionSize=0.1
    Debug=(bDrawDebug=false,DebugScale=25,DebugSegments=20,DebugThickness=0.25)
    DriverAxis=(X=0,Y=0,Z=-1)
    DriverItem=(Type=Bone,Name="clavicle_r")
    FalloffRegionScaleFactors=(PositiveWidth=1,NegativeWidth=1,PositiveHeight=1,NegativeHeight=1)
    FalloffSize=0.25
    OptionalParentItem=(Type=Bone)
    RotationOffset=(X=135,Y=0,Z=0)
   )
   > RigVMVariableNode(Value=@n20.OutputParam) {variable=spr_clavicle_r_twist_alpha}
  ]}
]
hierarchy [
 root {loc=<0 0 0>, rot=<0 0 0 1>}
 pelvis {parent=root, loc=<0 -0.5301 98.0571>, rot=<0.0127 0.707 -0.0127 -0.707>}
 spine_01 {parent=pelvis, loc=<6.1208 -0.9328 0>, rot=<0 0 0.034 -0.9994>}
 spine_02 {parent=spine_01, loc=<6.5996 -0.2095 0>, rot=<0 0 0.0238 -0.9997>}
 spine_03 {parent=spine_02, loc=<6.4914 -0.1237 0>, rot=<0 0 0.0149 -0.9999>}
 spine_04 {parent=spine_03, loc=<6.9598 0.0794 0>, rot=<0 0 -0.0132 -0.9999>}
 spine_05 {parent=spine_04, loc=<6.9987 0 0>, rot=<0 0 -0.0015 -1>}
 clavicle_l {parent=spine_05, loc=<5.4831 -0.1053 -3.2946>, rot=<0.0465 -0.7289 -0.009 -0.6829>}
 upperarm_l {parent=clavicle_l, loc=<9.8658 -1.2136 -0.4105>, rot=<0.0164 -0.3245 -0.0269 -0.9454>}
 lowerarm_l {parent=upperarm_l, loc=<26.465 -0.8345 0.259>, rot=<0.0036 0.0161 -0.2298 -0.9731>}
 hand_l {parent=lowerarm_l, loc=<22.7355 -0.3203 -0.6105>, rot=<-0.6784 -0.0223 0.0735 -0.7307>}
 index_metacarpal_l {parent=hand_l, loc=<1.6535 0.0183 -2.96>, rot=<0.0158 -0.0124 -0.0173 -0.9996>}
 index_01_l {parent=index_metacarpal_l, loc=<8.3008 0.0357 -0.0524>, rot=<-0.0044 0.0485 0.0924 -0.9945>}
 index_02_l {parent=index_01_l, loc=<3.3587 -0.3004 0.0452>, rot=<-0.002 0.0027 0.3225 -0.9466>}
 index_03_l {parent=index_02_l, loc=<2.5578 -0.123 -0.0252>, rot=<-0.002 0.0067 0.2558 -0.9667>}
 middle_metacarpal_l {parent=hand_l, loc=<2.1856 -0.1725 -0.5984>, rot=<0.006 0.0168 -0.0329 -0.9993>}
 middle_01_l {parent=middle_metacarpal_l, loc=<7.3275 -0.2447 0.0129>, rot=<-0.0208 0.0042 0.0858 -0.9961>}
 middle_02_l {parent=middle_01_l, loc=<4.0123 0.1401 0.1175>, rot=<-0.0218 0.0258 0.3512 -0.9357>}
 middle_03_l {parent=middle_02_l, loc=<2.6077 0.0534 0.1373>, rot=<0.0082 0.0188 0.1708 -0.9851>}
 pinky_metacarpal_l {parent=hand_l, loc=<1.7616 -0.9753 3.3814>, rot=<-0.0397 0.0814 -0.069 -0.9935>}
 pinky_01_l {parent=pinky_metacarpal_l, loc=<7.1709 -0.4831 -0.1094>, rot=<-0.0654 -0.0035 0.2242 -0.9723>}
 pinky_02_l {parent=pinky_01_l, loc=<2.8274 0.0893 -0.1043>, rot=<0.0037 0.0011 0.2854 -0.9584>}
 pinky_03_l {parent=pinky_02_l, loc=<2.2046 0.0307 -0.0078>, rot=<-0.0002 0.0098 0.1445 -0.9895>}
 ring_metacarpal_l {parent=hand_l, loc=<2.2074 -0.3625 1.3638>, rot=<-0.0284 0.0722 -0.0458 -0.9959>}
 ring_01_l {parent=ring_metacarpal_l, loc=<7.0295 -0.6197 -0.1446>, rot=<-0.0489 -0.0252 0.1892 -0.9804>}
 ring_02_l {parent=ring_01_l, loc=<3.6174 0.4647 -0.1027>, rot=<0 0 0.2806 -0.9598>}
 ring_03_l {parent=ring_02_l, loc=<2.4152 0.114 0.0199>, rot=<0 0 0.1189 -0.9929>}
 thumb_01_l {parent=hand_l, loc=<1.3118 -0.3191 -2.459>, rot=<0.6546 -0.4042 0.0664 -0.6354>}
 thumb_02_l {parent=thumb_01_l, loc=<5.9336 -0.0448 0.0401>, rot=<-0.0037 0.0213 0.298 -0.9543>}
 thumb_03_l {parent=thumb_02_l, loc=<2.2223 0.3682 -0.0907>, rot=<0.0118 -0.013 0.0066 -0.9998>}
 lowerarm_twist_01_l {parent=lowerarm_l, loc=<18.316 0 0>, rot=<0 0 0 1>}
 lowerarm_twist_02_l {parent=lowerarm_l, loc=<10.6707 0 0>, rot=<0 0 0 1>}
 upperarm_twist_02_l {parent=upperarm_l, loc=<19.7998 -0.1134 0>, rot=<0 0 0 1>}
 dfrm_shoulder_pad_mstr_l {parent=upperarm_l, loc=<0 0 0>, rot=<0.7008 -0.0076 0.0147 -0.7132>}
 dfrm_shoulder_pad_l {parent=dfrm_shoulder_pad_mstr_l, loc=<0 0 0>, rot=<0 0 0 1>}
 dfrm_shoulder_pad_up_aim_l {parent=clavicle_l, loc=<18.948 -0.7538 11.5186>, rot=<0.6791 -0.2195 -0.2326 -0.6607>}
 dfrm_clavicle_l {parent=clavicle_l, loc=<0 0 0>, rot=<0 0 0 1>}
 clavicle_r {parent=spine_05, loc=<5.4831 -0.1053 3.2946>, rot=<0.7289 0.0465 -0.6829 0.009>}
 upperarm_r {parent=clavicle_r, loc=<-9.8658 1.2136 0.4105>, rot=<0.0164 -0.3245 -0.0269 -0.9454>}
 lowerarm_r {parent=upperarm_r, loc=<-26.4651 0.8345 -0.259>, rot=<0.0036 0.0161 -0.2298 -0.9731>}
 hand_r {parent=lowerarm_r, loc=<-22.7356 0.3203 0.6106>, rot=<-0.6784 -0.0223 0.0735 -0.7307>}
 index_metacarpal_r {parent=hand_r, loc=<-1.6535 -0.0183 2.96>, rot=<0.0158 -0.0124 -0.0173 -0.9996>}
 index_01_r {parent=index_metacarpal_r, loc=<-8.3008 -0.0357 0.0524>, rot=<-0.0044 0.0485 0.0924 -0.9945>}
 index_02_r {parent=index_01_r, loc=<-3.3587 0.3004 -0.0452>, rot=<-0.002 0.0027 0.3225 -0.9466>}
 index_03_r {parent=index_02_r, loc=<-2.5578 0.123 0.0252>, rot=<-0.002 0.0067 0.2558 -0.9667>}
 middle_metacarpal_r {parent=hand_r, loc=<-2.1856 0.1725 0.5984>, rot=<0.006 0.0168 -0.0329 -0.9993>}
 middle_01_r {parent=middle_metacarpal_r, loc=<-7.3275 0.2447 -0.0129>, rot=<-0.0208 0.0042 0.0858 -0.9961>}
 middle_02_r {parent=middle_01_r, loc=<-4.0123 -0.1401 -0.1175>, rot=<-0.0218 0.0258 0.3512 -0.9357>}
 middle_03_r {parent=middle_02_r, loc=<-2.6077 -0.0533 -0.1373>, rot=<0.0082 0.0188 0.1708 -0.9851>}
 pinky_metacarpal_r {parent=hand_r, loc=<-1.7616 0.9753 -3.3814>, rot=<-0.0397 0.0814 -0.069 -0.9935>}
 pinky_01_r {parent=pinky_metacarpal_r, loc=<-7.1709 0.4831 0.1094>, rot=<-0.0654 -0.0035 0.2242 -0.9723>}
 pinky_02_r {parent=pinky_01_r, loc=<-2.8274 -0.0892 0.1043>, rot=<0.0037 0.0011 0.2854 -0.9584>}
 pinky_03_r {parent=pinky_02_r, loc=<-2.2046 -0.0307 0.0078>, rot=<-0.0002 0.0098 0.1445 -0.9895>}
 ring_metacarpal_r {parent=hand_r, loc=<-2.2074 0.3625 -1.3638>, rot=<-0.0284 0.0722 -0.0458 -0.9959>}
 ring_01_r {parent=ring_metacarpal_r, loc=<-7.0295 0.6197 0.1446>, rot=<-0.0489 -0.0252 0.1892 -0.9804>}
 ring_02_r {parent=ring_01_r, loc=<-3.6174 -0.4647 0.1027>, rot=<0 0 0.2806 -0.9598>}
 ring_03_r {parent=ring_02_r, loc=<-2.4153 -0.1141 -0.0199>, rot=<0 0 0.1189 -0.9929>}
 thumb_01_r {parent=hand_r, loc=<-1.3118 0.3191 2.459>, rot=<0.6546 -0.4042 0.0664 -0.6354>}
 thumb_02_r {parent=thumb_01_r, loc=<-5.9335 0.0447 -0.0401>, rot=<-0.0037 0.0213 0.298 -0.9543>}
 thumb_03_r {parent=thumb_02_r, loc=<-2.2223 -0.3682 0.0907>, rot=<0.0118 -0.013 0.0066 -0.9998>}
 lowerarm_twist_01_r {parent=lowerarm_r, loc=<-18.3161 0 0>, rot=<0 0 0 1>}
 lowerarm_twist_02_r {parent=lowerarm_r, loc=<-10.6707 0 0>, rot=<0 0 0 1>}
 upperarm_twist_02_r {parent=upperarm_r, loc=<-19.7998 0.1134 0>, rot=<0 0 0 1>}
 dfrm_shoulder_pad_mstr_r {parent=upperarm_r, loc=<0 0 0>, rot=<-0.0077 -0.7071 -0.7069 -0.0146>}
 dfrm_shoulder_pad_r {parent=dfrm_shoulder_pad_mstr_r, loc=<0 0 0>, rot=<0 0 0 1>}
 dfrm_shoulder_pad_up_aim_r {parent=clavicle_r, loc=<-18.9397 0.4862 -11.5116>, rot=<-0.2174 -0.685 -0.6546 0.2345>}
 dfrm_clavicle_r {parent=clavicle_r, loc=<0 0 0>, rot=<0 0 0 1>}
 neck_01 {parent=spine_05, loc=<11.3138 0.1529 0>, rot=<0 0 -0.2003 -0.9797>}
 neck_02 {parent=neck_01, loc=<4.5627 0 0>, rot=<0 0 0.0504 -0.9987>}
 head {parent=neck_02, loc=<5.3914 0 0>, rot=<0 0 0.115 -0.9934>}
 attach_cape {parent=spine_05, loc=<9.3216 -8.4501 0>, rot=<0.0616 -0.7044 -0.0616 -0.7044>}
 attach_backpack {parent=spine_05, loc=<0 -8.35 0>, rot=<-0.0093 -0.707 0.0093 -0.707>}
 dfrm_rib_mstr_01_r {parent=spine_05, loc=<0 0 0>, rot=<0.4324 -0.5674 0.5784 0.3956>}
 dfrm_rib_01_r {parent=dfrm_rib_mstr_01_r, loc=<15.8986 0 0>, rot=<0 -0.9999 -0.0166 0>}
 dfrm_rib_mstr_01_l {parent=spine_05, loc=<0 0 0>, rot=<0.4272 -0.575 -0.5709 -0.4013>}
 dfrm_rib_01_l {parent=dfrm_rib_mstr_01_l, loc=<15.8984 0 0>, rot=<0 -0.9996 0.0297 0>}
 dfrm_rib_mstr_03_l {parent=spine_03, loc=<0 0 0>, rot=<0.4702 -0.4405 -0.5229 -0.5581>}
 dfrm_rib_03_l {parent=dfrm_rib_mstr_03_l, loc=<13.1799 0 0>, rot=<0 -0.9985 0.0541 0>}
 dfrm_rib_mstr_03_r {parent=spine_03, loc=<0 0 0>, rot=<-0.4702 0.4405 -0.5229 -0.5581>}
 dfrm_rib_03_r {parent=dfrm_rib_mstr_03_r, loc=<13.1799 0 0>, rot=<0 -0.9985 -0.0541 0>}
 dfrm_rib_mstr_04_l {parent=spine_03, loc=<0 0 0>, rot=<0.3494 -0.4203 -0.644 -0.5353>}
 dfrm_rib_04_l {parent=dfrm_rib_mstr_04_l, loc=<11.9462 0 0>, rot=<0 -0.9987 0.0506 0>}
 dfrm_rib_mstr_04_r {parent=spine_03, loc=<0.0003 0 0>, rot=<-0.3494 0.4203 -0.644 -0.5353>}
 dfrm_rib_04_r {parent=dfrm_rib_mstr_04_r, loc=<11.9463 0 0>, rot=<0 -0.9987 -0.0506 0>}
 dfrm_spine_04 {parent=spine_03, loc=<6.9598 0.0794 0>, rot=<0 0 0.0843 -0.9964>}
 dfrm_rib_mstr_02_l {parent=dfrm_spine_04, loc=<0 0 0>, rot=<-0.4444 0.5223 0.5403 0.4877>}
 dfrm_rib_02_l {parent=dfrm_rib_mstr_02_l, loc=<14.0951 0 0>, rot=<0 -0.9984 0.0565 0>}
 dfrm_rib_mstr_02_r {parent=dfrm_spine_04, loc=<0 0 0>, rot=<0.4414 -0.5255 0.5371 0.4904>}
 dfrm_rib_02_r {parent=dfrm_rib_mstr_02_r, loc=<14.0951 0 0>, rot=<0 -0.998 -0.0626 0>}
 thigh_l {parent=pelvis, loc=<-1.0855 -0.2109 -8.7053>, rot=<0.0009 0.0355 0.0155 -0.9993>}
 calf_l {parent=thigh_l, loc=<-42.5284 -0.0019 -0.7297>, rot=<0.0006 -0.0161 -0.0391 -0.9991>}
 calf_twist_01_l {parent=calf_l, loc=<-35.8483 -1.9197 -0.5794>, rot=<0.0237 0.0095 -0.0035 -0.9997>}
 calf_twist_02_l {parent=calf_l, loc=<-19.1546 -1.7943 -0.2638>, rot=<0.0237 0.0095 -0.0035 -0.9997>}
 foot_l {parent=calf_l, loc=<-46.7402 -2.5014 -0.592>, rot=<0.0427 -0.0238 0.0353 -0.9982>}
 ball_l {parent=foot_l, loc=<-3.5658 10.8066 -0.1544>, rot=<0.0001 -0.0001 -0.7063 -0.7079>}
 thigh_r {parent=pelvis, loc=<-1.0855 -0.2109 8.7053>, rot=<0.0355 -0.0009 0.9993 0.0155>}
 calf_r {parent=thigh_r, loc=<42.5284 0.0019 0.7297>, rot=<0.0006 -0.0161 -0.0391 -0.9991>}
 calf_twist_01_r {parent=calf_r, loc=<35.8483 1.9197 0.5794>, rot=<0.0237 0.0095 -0.0035 -0.9997>}
 calf_twist_02_r {parent=calf_r, loc=<19.1546 1.7943 0.2638>, rot=<0.0237 0.0095 -0.0035 -0.9997>}
 foot_r {parent=calf_r, loc=<46.7401 2.5014 0.592>, rot=<0.0427 -0.0238 0.0353 -0.9982>}
 ball_r {parent=foot_r, loc=<3.5658 -10.8066 0.1544>, rot=<0.0001 -0.0001 -0.7063 -0.7079>}
 dfrm_thigh_r {parent=pelvis, loc=<-1.0855 -0.2109 8.7053>, rot=<0.044 -0.001 0.9989 0.0155>}
 dfrm_hip_adjust_r {parent=dfrm_thigh_r, loc=<5.7251 0.7177 3.4247>, rot=<0.0695 -0.6734 0.7333 0.0627>}
 dfrm_thigh_l {parent=pelvis, loc=<-1.0855 -0.2109 -8.7053>, rot=<-0.044 0.001 0.9989 0.0155>}
 dfrm_hip_adjust_l {parent=dfrm_thigh_l, loc=<5.7251 0.7177 -3.4247>, rot=<-0.1305 0.6643 0.7246 0.1293>}
 dfrm_thigh_up_aim_l {parent=pelvis, loc=<0.2338 -0.2623 -23.6471>, rot=<-0.044 0.001 0.9989 0.0155>}
 dfrm_thigh_up_aim_r {parent=pelvis, loc=<0.2338 -0.2623 23.6471>, rot=<0.044 -0.001 0.9989 0.0155>}
]
variables [
 has_rifle {type=bool, default=False}
 is_crouching {type=bool, default=False}
 is_frontEnd {type=bool, default=False}
 is_skydiving {type=bool, default=False}
 look_down_rib_mstr_02_l_rot {type=FVector, default=(X=0,Y=0,Z=0)}
 look_down_rib_mstr_03_l_rot {type=FVector, default=(X=0,Y=0,Z=0)}
 spr_ADS_pistol_up_alpha {type=double, default=0}
 spr_calvicle_l_up_alpha {type=double, default=0}
 spr_calvicle_r_up_alpha {type=double, default=0}
 spr_clavicle_r_twist_alpha {type=double, default=0}
 spr_thigh_l_crouch_alpha {type=double, default=0}
 spr_thigh_r_crouch_alpha {type=double, default=0}
 spr_upperarm_l_clavicle_r_aim_dwn_alpha {type=double, default=0}
 spr_upperarm_l_down_alpha {type=double, default=0}
 spr_upperarm_l_skydive_alpha {type=double, default=0}
 spr_upperarm_r_down_alpha {type=double, default=0}
 spr_upperarm_r_skydive_alpha {type=double, default=0}
 upper_arm_dwn_rib_mstr_scl_fctr {type=FVector, default=(X=0.9,Y=0.9,Z=0.9)}
]
