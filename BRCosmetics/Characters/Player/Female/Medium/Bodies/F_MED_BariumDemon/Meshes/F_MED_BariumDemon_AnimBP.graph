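# Character part AnimBP for the F_MED_BariumDemon body, layering procedural bone adjustments onto the pose copied from its attached parent mesh.
# The copied pose runs through the body's Control Rig, fed crouch, skydive, front-end and has_rifle states.
# A constraint then drives attach_cape from spine_05 and both clavicles, weighted heavily toward the spine.
# Each update resets has_rifle, then sets it for rifle, shotgun, sniper and rocket launcher animation types.
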AnimBlueprint /BRCosmetics/Characters/Player/Female/Medium/Bodies/F_MED_BariumDemon/Meshes/F_MED_BariumDemon_AnimBP
target_skeleton=&/F_MED_BariumDemon_Skeleton
graphs [
 AnimGraph {nodes [
   n1 = Control Rig!(
    Source=Copy Pose From Mesh() {bUseAttachedParent=true}
    Alpha
    has_rifle:bool=Get has_rifle() -> (has_rifle:bool)
    is_crouching:bool=Get bIsCrouching()
    is_skydiving:bool=Get bIsSkyDiving()
    is_frontEnd:bool=Get bIsInFrontEnd()
   ) -> (Pose) {
    alphaBoolBlend {blendOption=Linear}
    controlRigAssetReference {blueprintRigClass=&/F_MED_BariumDemon_CtrlRig.F_MED_BariumDemon_CtrlRig_C}
   }
   n2 = "Constraint - attach_cape (PPP)"!(
    ComponentPose=Local To Component(LocalPose=@n1)
    Alpha=1
    ConstraintWeights_0=0.9203
    ConstraintWeights_1:float=0.045
    ConstraintWeights_2:float=0.0347
   ) -> (Pose) {
    alphaBoolBlend {blendOption=Linear}
    boneToModify {boneName=attach_cape}
    constraintSetup [{targetBone {boneName=spine_05}}, {targetBone {boneName=clavicle_r}}, {targetBone {boneName=clavicle_l}}]
    constraintWeights [1,1,1]
   }
   Output Pose(Result=Component To Local(ComponentPose=@n2)) {layerGroup=None}
  ]}
 EventGraph {nodes [
   Event BlueprintUpdateAnimation()
   > Set has_rifle(has_rifle:bool=false) -> (Output_Get:bool)
   > n1 = Switch on EFortWeaponCoreAnimation!(
    execute
    Selection:EFortWeaponCoreAnimation Enum=GetWeaponCoreAnim!(self:Fort Player Animation Interface Object Reference=GetOwnerAnimationInterface!(self:Custom Character Part Anim Instance Object Reference) -> (ReturnValue:Fort Player Animation Interface Object Reference)) -> (ReturnValue:EFortWeaponCoreAnimation Enum)
   ) -> (
    Melee:Exec
    Pistol:Exec
    Shotgun:Exec
    PaperBlueprint:Exec
    Rifle:Exec
    MeleeOneHand:Exec
    MachinePistol:Exec
    RocketLauncher:Exec
    GrenadeLauncher:Exec
    GoingCommando:Exec
    AssaultRifle:Exec
    TacticalShotgun:Exec
    SniperRifle:Exec
    TrapPlacement:Exec
    ShoulderLauncher:Exec
    AbilityDecoTool:Exec
    Crossbow:Exec
    C4:Exec
    RemoteControl:Exec
    DualWield:Exec
    AR_BullPup:Exec
    AR_ForwardGrip:Exec
    MedPackPaddles:Exec
    SMG_P90:Exec
    AR_DrumGun:Exec
    Consumable_Small:Exec
    Consumable_Large:Exec
    Balloon:Exec
    MountedTurret:Exec
    CreativeTool:Exec
    ExplosiveBow:Exec
    AshtonIndigo:Exec
    AshtonChicago:Exec
    MeleeDualWield:Exec
    Unarmed:Exec
   )
   Branch(execute=[@n1.AssaultRifle, @n1.Rifle, @n1.RocketLauncher, @n1.Shotgun, @n1.SniperRifle, @n1.TacticalShotgun])
   > Set has_rifle(has_rifle=true)
  ]}
]
variables [has_rifle {type=boolean}]
